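# Parent cloth-model material for the CephaloChef character body, layering animated paper detail onto the shared 2023 anime shader.
# Paper texture channels scroll on ShakeCrunch stepped time with derivative-based mips, tinting lit and shaded colors via overlay and Fresnel blends.
# Perlin noise builds a three-step shading curve, alongside view-dependent backlighting and threshold controls.
# A static switch adds the anime flipbook face and its vertex-hide offset.
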Material /BRCosmetics/Characters/Player/Male/Medium/Bodies/M_MED_CephaloChef/Materials/M_CephaloChef_Parent
graph {
 properties {
  bAutomaticallySetUsageInEditor=false
  bHasPixelAnimation=true
  bUseMaterialAttributes=true
  bUsedWithMeshDeformer=true
  bUsedWithMorphTargets=true
  bUsedWithSkeletalMesh=true
  bUsedWithStaticLighting=true
  shadingModel=MSM_Cloth
  usedShadingModels=Cloth
 }
 nodes [
  Constant()
  n1 = Constant() {r=0.5}
  n2 = Multiply(A=Constant() {r=0.1}) {constB=0.5}
  n3 = Subtract(A=@n1, B=@n2) {constA=0.5}
  n4 = Add(A=@n2, B=@n1) {constB=0.5}
  n5 = Multiply(A=TextureCoordinate(), B=ScalarParameter() {defaultValue=2, group=03_CephaloChef, parameterName=Noise_Scale})
  n6 = TextureSample(UVs=Multiply(A=@n5) {constB=1.8}) {samplerSource=SSM_Wrap_WorldGroupSettings, samplerType=SAMPLERTYPE_LinearColor, texture=/EpicBaseTextures/Noises/T_Perlin_Noise_D}
  n7 = TextureSample(UVs=@n5) {samplerSource=SSM_Wrap_WorldGroupSettings, samplerType=SAMPLERTYPE_LinearColor, texture=&T_Perlin_Noise_D}
  n8 = Subtract(A=@n7.R) {constB=0.5}
  n9 = AppendVector(
   A=Multiply(A=Subtract(A=@n6.R) {constB=0.5}) {constB=0.1}
   B=NamedRerouteDeclaration(None=Multiply(A=@n8) {constB=0.2}) {name=Noise, nodeColor=<rgba 1 0.6563 0 1>}
  )
  n10 = TextureSample(UVs=Multiply(A=@n5) {constB=0.8}) {samplerSource=SSM_Wrap_WorldGroupSettings, samplerType=SAMPLERTYPE_LinearColor, texture=&T_Perlin_Noise_D}
  n11 = Multiply(A=Subtract(A=@n10.R) {constB=0.5}) {constB=0.3}
  MaterialExpressionAdd_5 = Add(
   A=Add(A=NamedRerouteUsage(), B=AppendVector(A=@n9, B=@n11))
   B=ScalarParameter() {group=03_CephaloChef, parameterName=ShadingCurve_Bias, sliderMax=1, sliderMin=-1}
  )
  n12 = Convert(None=@MaterialExpressionAdd_5) {
   nodeName=Break Float3
   convertInputs [{
     type=Vector3
     expressionInput {
      expression=&self:MaterialExpressionAdd_5
      inputName=None
      mask=0
      maskA=0
      maskB=0
      maskG=0
      maskR=0
      outputIndex=0
     }
    }]
   convertOutputs [{type=Scalar}, {type=Scalar}, {type=Scalar}]
  }
  n13 = VectorParameter() {group=03_CephaloChef, parameterName=ShadingCurve_Offset, defaultValue=<rgba 0 0.1 0.3 0>}
  n14 = Add(A=@n12, B=@n13.R) {constB=-0.2}
  n15 = Add(A=@n12, B=@n13.G) {constB=0}
  n16 = Add(
   A=Multiply(A=SmoothStep(Min=@n3, Max=@n4, Value=@n14)) {constB=0.3333}
   B=Multiply(A=SmoothStep(Min=@n3, Max=@n4, Value=@n15)) {constB=0.3333}
  )
  n17 = Multiply(A=SmoothStep(Min=@n3, Max=@n4, Value=Add(A=@n12, B=@n13.B) {constB=0.3})) {constB=0.3333}
  NamedRerouteDeclaration(None=Saturate(None=Add(A=@n16, B=@n17))) {name=ShadingCurve, nodeColor=<rgba 0 1 0.0625 1>}
  n18 = MaterialFunctionCall!(Mouth UV, Eye UV, Brow UV, Material) {materialFunction=/Game/Global/Materials/Character_Parent_2023/FlipbookFace_Parent_2023/MaterialFunctions/MFI_FlipbookFace_Anime}
  NamedRerouteDeclaration(None=GetMaterialAttributes(None=@n18.Material) {attributeGetTypes [#1]}.World Position Offset) {name=WPO VertHide, nodeColor=<rgba 0.5234 0 1 1>}
  NamedRerouteDeclaration(None=MaterialFunctionCall() {materialFunction=/Game/Global/Materials/Character_Parent_2023/Anime_Parent_2023/MaterialFunctions/MFC_ActorAlighnedScreenPosition}.UVs) {name=ObjectAlighedScreenPosition, nodeColor=<rgba 1 0.867 0 1>}
  n19 = MaterialFunctionCall!(CustomUV) {materialFunction=/Game/Global/Materials/Character_Parent_2023/Anime_Parent_2023/MaterialFunctions/MFC_AnimeShader_2023_Pre}
  n20 = MaterialFunctionCall!(CustomLightVector, OverrideNormal_WS, STT=@n19.STT, Normal_Map=@n19.STT, DFL=@n19.STT, FX=@n19.STT) {materialFunction=/Game/Global/Materials/Character_Parent_2023/Anime_Parent_2023/MaterialFunctions/MFC_AnimeShader_2023_Calculations}
  NamedRerouteDeclaration(None=DotProduct(A=@n20.Normal_WS, B=CameraVectorWS())) {name=NdotV, nodeColor=<rgba 1 0.6242 0 1>}
  NamedRerouteDeclaration(None=DotProduct(
    A=MaterialFunctionCall() {materialFunction=/Game/Global/Materials/Character_Parent_2023/Anime_Parent_2023/MaterialFunctions/MFC_AnimeShader_LightDirection}.Direction Vector
    B=PixelNormalWS()
   )) {name=NdotL, nodeColor=<rgba 0.9723 1 0.2825 1>}
  n21 = NamedRerouteUsage()
  n22 = Add(A=NamedRerouteUsage(), B=NamedRerouteUsage())
  n23 = Add(A=ScalarParameter() {defaultValue=4, group=02_Anime Parameters, parameterName=BackLit_Exponent, sliderMax=12, sortPriority=31}) {constB=2.2}
  n24 = Saturate(None=SmoothStep(Value=Power(Base=@n22, Exp=@n23) {constExponent=4}) {constMax=0, constMin=0.02})
  n25 = Add(
   A=DotProduct(
    A=MaterialFunctionCall() {materialFunction=&CameraDirectionVector}.Result
    B=MaterialFunctionCall() {materialFunction=&MFC_AnimeShader_LightDirection}.Direction Vector
   )
   B=ScalarParameter() {defaultValue=0.25, group=02_Anime Parameters, parameterName=BackLit_Bias, sliderMax=1}
  )
  n26 = ScalarParameter() {defaultValue=0.75, group=02_Anime Parameters, parameterName=BackLit_MaxBias, sliderMax=1}
  n27 = Multiply(A=LinearInterpolate(A=@n21, B=Add(A=@n21, B=@n24), Alpha=Clamp(None=@n25, Max=@n26))) {constB=0.5}
  NamedRerouteDeclaration(None=MaterialFunctionCall!(Base=NamedRerouteUsage(), Blend=Add(A=@n27) {constB=0.5}) {materialFunction=/Game/Global/Materials/Character_Parent_2023/Anime_Parent_2023/MaterialFunctions/MF_BlendOverlay}.Result) {name=BackLitShading+Threshold, nodeColor=<rgba 0 1 0.9234 1>}
  n28 = LinearInterpolate(
   A=ScalarParameter() {group=02_Anime Parameters, parameterName=Threshold_Min, sliderMax=1}
   B=ScalarParameter() {defaultValue=1, group=02_Anime Parameters, parameterName=Threshold_Max, sliderMax=1}
   Alpha=Power(Base=@n20.Threshold_Gradient, Exp=ScalarParameter() {defaultValue=1, group=02_Anime Parameters, parameterName=Threshold_Exp, sliderMax=8})
  )
  n29 = NamedRerouteDeclaration(None=StaticBoolParameter() {group=03_CephaloChef, parameterName=UseFlipBookFace, sortPriority=0}) {name=FlipBookBool, nodeColor=<rgba 0 1 0.2031 1>}
  n30 = StaticSwitch(True=LinearInterpolate(A=@n19.Color_Lit, B=@n18.FaceColor, Alpha=@n18.FaceColor), False=@n19.Color_Lit, Value=@n29)
  n31 = Constant() {r=0.5}
  n32 = NamedRerouteUsage()
  n33 = Multiply(A=@n32, B=ScalarParameter() {defaultValue=0.5, group=03_CephaloChef, parameterName=PaperDetail_TilingOverall, sliderMax=2})
  n34 = Multiply(A=@n33, B=ScalarParameter() {defaultValue=0.5, group=03_CephaloChef, parameterName=PaperDetail_Tiling_RChannel, sliderMax=12})
  n35 = MaterialFunctionCall() {materialFunction=/BRCosmetics/Characters/Player/Male/Medium/Bodies/M_MED_ShakeCrunch/Materials/MFC_ShakeCrunch_SteppedTime}
  n36 = DDX(Value=@n32)
  n37 = DDY(Value=@n32)
  n38 = TextureSample!(UVs=Frac(None=Add(A=@n34, B=@n35.TimeStep)), Tex, "DDX(UVs)"=@n36, "DDY(UVs)"=@n37, Apply View MipBias) {
   mipValueMode=TMVM_Derivative
   samplerSource=SSM_Wrap_WorldGroupSettings
   samplerType=SAMPLERTYPE_LinearColor
   texture=/BRCosmetics/Characters/Player/Male/Medium/Bodies/M_MED_CephaloChef/Textures/T_CephaloChef_PaperTextures
  }
  n39 = Multiply(A=@n38.R, B=ScalarParameter() {defaultValue=0.5, group=03_CephaloChef, parameterName=PaperDetail_Intensity_RChannel, sliderMax=1}) {constB=0.7}
  n40 = Multiply(A=@n33, B=ScalarParameter() {defaultValue=1, group=03_CephaloChef, parameterName=PaperDetail_Tiling_GChannel, sliderMax=12})
  n41 = TextureSample(UVs=Frac(None=Add(A=@n40, B=@n35.TimeStep)), "DDX(UVs)"=@n36, "DDY(UVs)"=@n37) {
   mipValueMode=TMVM_Derivative
   samplerSource=SSM_Wrap_WorldGroupSettings
   samplerType=SAMPLERTYPE_LinearColor
   texture=&T_CephaloChef_PaperTextures
  }
  n42 = ScalarParameter() {defaultValue=0.5, group=03_CephaloChef, parameterName=PaperDetail_Intensity_GChannel, sliderMax=1}
  n43 = Multiply(A=@n33, B=ScalarParameter() {defaultValue=1, group=03_CephaloChef, parameterName=PaperDetail_Tiling_BChannel, sliderMax=12})
  n44 = TextureSample(UVs=Frac(None=Add(A=@n43, B=@n35.TimeStep)), "DDX(UVs)"=@n36, "DDY(UVs)"=@n37) {
   mipValueMode=TMVM_Derivative
   samplerSource=SSM_Wrap_WorldGroupSettings
   samplerType=SAMPLERTYPE_LinearColor
   texture=&T_CephaloChef_PaperTextures
  }
  n45 = Add(
   A=Add(A=Add(A=@n31, B=@n39), B=Multiply(A=@n41.G, B=@n42) {constB=0.4})
   B=Multiply(
    A=Subtract(A=@n44.B) {constB=0.5}
    B=ScalarParameter() {defaultValue=1, group=03_CephaloChef, parameterName=PaperDetail_Intensity_BChannel, sliderMax=1}
   ) {constB=0.5}
  )
  n46 = Multiply(A=@n33, B=ScalarParameter() {defaultValue=1, group=03_CephaloChef, parameterName=PaperDetail_Tiling_Weave, sliderMax=64})
  n47 = TextureSample(UVs=Frac(None=Add(A=@n46, B=@n35.TimeStep)), "DDX(UVs)"=@n36, "DDY(UVs)"=@n37) {
   mipValueMode=TMVM_Derivative
   samplerSource=SSM_Wrap_WorldGroupSettings
   samplerType=SAMPLERTYPE_LinearColor
   texture=&T_CephaloChef_PaperTextures
  }
  n48 = Multiply(
   A=Subtract(A=@n47.B) {constB=0.5}
   B=ScalarParameter() {defaultValue=1, group=03_CephaloChef, parameterName=PaperDetail_Intensity_Weave, sliderMax=1}
  ) {constB=0.5}
  n49 = Multiply(
   A=Add(A=Add(A=@n45, B=@n48), B=ScalarParameter() {group=03_CephaloChef, parameterName=PaperDetail_Bias, sliderMax=1, sliderMin=-1})
   B=VectorParameter() {group=03_CephaloChef, parameterName=Paper_Tint, defaultValue=<rgba 1 1 1 0>}.RGB
  )
  n50 = MaterialFunctionCall(Base=@n30, Blend=@n49) {materialFunction=&Blend_Overlay}
  n51 = ScalarParameter() {defaultValue=3, group=03_CephaloChef, parameterName=PaperDetail_Fresnel_Exp, sliderMax=8}
  n52 = ScalarParameter() {group=03_CephaloChef, parameterName=PaperDetail_Fresnel_BaseFraction, sliderMax=1}
  n53 = Saturate(None=Multiply(
    A=Fresnel(ExponentIn=@n51, BaseReflectFractionIn=@n52) {baseReflectFraction=0, exponent=3}
    B=ScalarParameter() {defaultValue=1, group=03_CephaloChef, parameterName=PaperDetail_Fresnel_Instensity, sliderMax=8}
   ))
  n54 = ScalarParameter() {defaultValue=1, group=03_CephaloChef, parameterName=PaperDetail_Blend, sliderMax=1, sortPriority=1}
  n55 = ScalarParameter() {defaultValue=1, group=03_CephaloChef, parameterName=FlipbookFaceAnimeShadedValue}
  n56 = StaticSwitch(True=LinearInterpolate(A=@n19.Color_Shaded, B=Multiply(A=@n18.FaceColor, B=@n55), Alpha=@n18.FaceAlpha), False=@n19.Color_Shaded, Value=@n29)
  n57 = MaterialFunctionCall(Base=@n56, Blend=@n49) {materialFunction=&Blend_Overlay}
  n58 = MaterialFunctionCall!(
   Shading_Gradient=NamedRerouteUsage()
   Threshold_Gradient=NamedRerouteDeclaration(None=Saturate(None=@n28)) {name=Threshold, nodeColor=<rgba 1 0.2971 0.8877 1>}
   Specular_Gradient=@n20.Specular_Gradient
   Sheen_Mask=@n20.Specular_Gradient
   Thickness=@n20.Specular_Gradient
   DFL_Mask=@n20.Specular_Gradient
   FX=@n20.Specular_Gradient
   Metallic_Gradient=@n20.Specular_Gradient
   RimLight_Gradient=@n20.Specular_Gradient
   NormalBuffer=@n20.Specular_Gradient
   --------
   CINE_LightSources=@n20.Specular_Gradient
   CINE_LightColor=@n20.Specular_Gradient
   ---------
   Color_Lit=LinearInterpolate(A=@n30, B=LinearInterpolate(A=@n50.Result, B=@n49, Alpha=@n53), Alpha=@n54)
   Color_Shaded=LinearInterpolate(A=@n56, B=LinearInterpolate(A=@n57.Result, B=@n49, Alpha=@n53), Alpha=@n54)
   SRM=@n19.SRM
   Emissive=@n19.SRM
   DFL_Color=@n19.SRM
   Specular_Color=@n19.SRM
   RimLight_Color=@n19.SRM
  ) {materialFunction=/Game/Global/Materials/Character_Parent_2023/Anime_Parent_2023/MaterialFunctions/MFC_AnimeShader_2023_Post}
  n59 = GetMaterialAttributes(None=@n58.Result) {attributeGetTypes [#1]}
  n60 = NamedRerouteUsage()
  StaticSwitch(
   True=SetMaterialAttributes!(MaterialAttributes=@n59.MaterialAttributes, World Position Offset=Add(A=@n59.World Position Offset, B=@n60)) {attributeSetTypes [#1]}
   False=@n58.Result
   Value=NamedRerouteUsage()
  )
 ]
}
