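# Control Rig for the CephaloChef character body that procedurally drives its tentacle arms, tentacle index fingers and extra chest and forearm joints.
# Tentacle chains are fitted to stretched Hermite splines built from arm, hand and finger bones, with parent constraints anchoring helper joints.
# Spherical pose readers on upper arms and index fingers weight corrective offsets, many gated by rifle, pistol and not-playing-emote variables.
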ControlRigBlueprint /BRCosmetics/Characters/Player/Male/Medium/Bodies/M_MED_CephaloChef/Meshes/M_MED_CephaloChef_CtrlRig
graphs [
 Add_1_ContainedGraph {nodes [
   n1 = RigVMFunctionEntryNode()
   Add = &RigVMFunction_MathDoubleAdd(A=@n1.A, B=@n1.B)
   RigVMFunctionReturnNode(Result=&RigVMFunction_MathDoubleAdd(A=@Add.Result, B=@n1.C).Result)
  ]}
 Add_1_ContainedGraph {nodes [
   n1 = RigVMFunctionEntryNode()
   Add = &RigVMFunction_MathDoubleAdd(A=@n1.A, B=@n1.B)
   RigVMFunctionReturnNode(Result=&RigVMFunction_MathDoubleAdd(A=@Add.Result, B=@n1.C).Result)
  ]}
 Multiply_2_ContainedGraph {nodes [
   n1 = RigVMFunctionEntryNode()
   Multiply = &RigVMFunction_MathDoubleMul(A=@n1.A, B=@n1.B)
   RigVMFunctionReturnNode(Result=&RigVMFunction_MathDoubleMul(B=@n1.C, A=@Multiply.Result).Result)
  ]}
 Multiply_2_ContainedGraph {nodes [
   n1 = RigVMFunctionEntryNode()
   Multiply = &RigVMFunction_MathDoubleMul(A=@n1.A, B=@n1.B)
   RigVMFunctionReturnNode(Result=&RigVMFunction_MathDoubleMul(B=@n1.C, A=@Multiply.Result).Result)
  ]}
 Multiply_2_ContainedGraph {nodes [
   n1 = RigVMFunctionEntryNode()
   Multiply = &RigVMFunction_MathDoubleMul(A=@n1.A, B=@n1.B)
   RigVMFunctionReturnNode(Result=&RigVMFunction_MathDoubleMul(B=@n1.C, A=@Multiply.Result).Result)
  ]}
 Multiply_2_ContainedGraph {nodes [
   n1 = RigVMFunctionEntryNode()
   Multiply = &RigVMFunction_MathDoubleMul(A=@n1.A, B=@n1.B)
   RigVMFunctionReturnNode(Result=&RigVMFunction_MathDoubleMul(B=@n1.C, A=@Multiply.Result).Result)
  ]}
 Multiply_2_ContainedGraph {nodes [
   n1 = RigVMFunctionEntryNode()
   Multiply = &RigVMFunction_MathDoubleMul(A=@n1.A, B=@n1.B)
   RigVMFunctionReturnNode(Result=&RigVMFunction_MathDoubleMul(B=@n1.C, A=@Multiply.Result).Result)
  ]}
 Multiply_2_ContainedGraph {nodes [
   n1 = RigVMFunctionEntryNode()
   Multiply = &RigVMFunction_MathDoubleMul(A=@n1.A, B=@n1.B)
   RigVMFunctionReturnNode(Result=&RigVMFunction_MathDoubleMul(B=@n1.C, A=@Multiply.Result).Result)
  ]}
 Multiply_2_ContainedGraph {nodes [
   n1 = RigVMFunctionEntryNode()
   Multiply = &RigVMFunction_MathDoubleMul(A=@n1.A, B=@n1.B)
   RigVMFunctionReturnNode(Result=&RigVMFunction_MathDoubleMul(B=@n1.C, A=@Multiply.Result).Result)
  ]}
 Multiply_2_ContainedGraph {nodes [
   n1 = RigVMFunctionEntryNode()
   Multiply = &RigVMFunction_MathDoubleMul(A=@n1.A, B=@n1.B)
   RigVMFunctionReturnNode(Result=&RigVMFunction_MathDoubleMul(B=@n1.C, A=@Multiply.Result).Result)
  ]}
 Multiply_2_ContainedGraph {nodes [
   n1 = RigVMFunctionEntryNode()
   Multiply = &RigVMFunction_MathDoubleMul(A=@n1.A, B=@n1.B)
   RigVMFunctionReturnNode(Result=&RigVMFunction_MathDoubleMul(B=@n1.C, A=@Multiply.Result).Result)
  ]}
 Multiply_2_ContainedGraph {nodes [
   n1 = RigVMFunctionEntryNode()
   Multiply = &RigVMFunction_MathDoubleMul(A=@n1.A, B=@n1.B)
   RigVMFunctionReturnNode(Result=&RigVMFunction_MathDoubleMul(B=@n1.C, A=@Multiply.Result).Result)
  ]}
 Multiply_2_ContainedGraph {nodes [
   n1 = RigVMFunctionEntryNode()
   Multiply = &RigVMFunction_MathDoubleMul(A=@n1.A, B=@n1.B)
   RigVMFunctionReturnNode(Result=&RigVMFunction_MathDoubleMul(B=@n1.C, A=@Multiply.Result).Result)
  ]}
 Multiply_2_ContainedGraph {nodes [
   n1 = RigVMFunctionEntryNode()
   Multiply = &RigVMFunction_MathDoubleMul(A=@n1.A, B=@n1.B)
   RigVMFunctionReturnNode(Result=&RigVMFunction_MathDoubleMul(B=@n1.C, A=@Multiply.Result).Result)
  ]}
 New Function_ContainedGraph {nodes [
   RigVMFunctionEntryNode()
   > RigVMFunction_Sequence = RigVMAggregateNode()
   n1 = &RigUnit_ItemArray(Items=((Type=Bone,Name="cr_index_metacarpal_l"),(Type=Bone,Name="cr_index_01_l"),(Type=Bone,Name="cr_index_02_l"),(Type=Bone,Name="cr_index_03_l")))
   For_Each = RigVMDispatchNode(Array=@n1.Items, ExecuteContext=@RigVMFunction_Sequence.A) {template="DISPATCH_RigVMDispatch_ArrayIterator(in Array,out Element,out Index,out Count,out Ratio)"}
   n2 = &RigUnit_ItemArray(Items=((Type=Bone,Name="index_metacarpal_l"),(Type=Bone,Name="index_01_l"),(Type=Bone,Name="index_02_l"),(Type=Bone,Name="index_03_l")))
   > &RigUnit_ParentConstraint(
    AdvancedSettings=(InterpolationType=Average,RotationOrderForFilter=XZY)
    Parents=((Item=(Type=Bone,Name="None"),Weight=1))
    Weight=1
    bMaintainOffset=false
    Parents.0.Item=RigVMDispatchNode(Index=@For_Each.Index, Array=@n2.Items) {template="DISPATCH_RigVMDispatch_ArrayGetAtIndex(in Array,in Index,out Element)"}.Element
    Child=@For_Each.Element
   )
   n3 = &RigUnit_ItemArray(Items=((Type=Bone,Name="cr_index_metacarpal_r"),(Type=Bone,Name="cr_index_01_r"),(Type=Bone,Name="cr_index_02_r"),(Type=Bone,Name="cr_index_03_r")))
   For_Each_1 = RigVMDispatchNode(ExecuteContext=@For_Each.Completed, Array=@n3.Items) {template="DISPATCH_RigVMDispatch_ArrayIterator(in Array,out Element,out Index,out Count,out Ratio)"}
   n4 = &RigUnit_ItemArray(Items=((Type=Bone,Name="index_metacarpal_r"),(Type=Bone,Name="index_01_r"),(Type=Bone,Name="index_02_r"),(Type=Bone,Name="index_03_r")))
   > &RigUnit_ParentConstraint(
    AdvancedSettings=(InterpolationType=Average,RotationOrderForFilter=XZY)
    Parents=((Item=(Type=Bone,Name="None"),Weight=1))
    Weight=1
    bMaintainOffset=false
    Parents.0.Item=RigVMDispatchNode(Index=@For_Each_1.Index, Array=@n4.Items) {template="DISPATCH_RigVMDispatch_ArrayGetAtIndex(in Array,in Index,out Element)"}.Element
    Child=@For_Each_1.Element
   )
   n5 = &RigUnit_SphericalPoseReader(
    ActiveRegionScaleFactors=(PositiveWidth=4,NegativeWidth=2,PositiveHeight=4,NegativeHeight=4)
    ActiveRegionSize=0.1
    Debug=(bDrawDebug=false,bDraw2D=False,bDrawLocalAxes=False,DebugScale=25,DebugSegments=20,DebugThickness=0.25)
    DriverAxis=(X=0,Y=1,Z=0)
    DriverItem=(Type=Bone,Name="index_03_l")
    FalloffRegionScaleFactors=(PositiveWidth=1,NegativeWidth=1,PositiveHeight=1,NegativeHeight=1)
    FalloffSize=0.4
    FlipHeightScaling=False
    FlipWidthScaling=False
    OptionalParentItem=(Type=Bone,Name="hand_l")
    RotationOffset=(X=90,Y=20,Z=-72)
    ExecutePin=@RigVMFunction_Sequence.B
   )
   > RigUnit_SetRelativeTransformForItem = /Script/ControlRig.RigUnit_SetRelativeTransformForItem(
    Child=(Type=Bone,Name="cr_index_03_l")
    Parent=(Type=Bone,Name="hand_l")
    bParentInitial=False
    bPropagateToChildren=True
    Value=&RigUnit_GetRelativeTransformForItem(Child=(Type=Bone,Name="index_03_l"), Parent=(Type=Bone,Name="hand_l"), bChildInitial=False, bParentInitial=False).RelativeTransform
    Weight=@n5.OutputParam
   ) {
    is=struct RigVMUnitNode
    template="Set Relative Transform::Execute(in Child,in Parent,in bParentInitial,in Value,in Weight,in bPropagateToChildren)"
   }
   > n6 = &RigUnit_SphericalPoseReader(
    ActiveRegionScaleFactors=(PositiveWidth=4,NegativeWidth=3,PositiveHeight=4,NegativeHeight=4)
    ActiveRegionSize=0.1
    Debug=(bDrawDebug=false,bDraw2D=False,bDrawLocalAxes=False,DebugScale=25,DebugSegments=20,DebugThickness=0.25)
    DriverAxis=(X=0,Y=-1,Z=0)
    DriverItem=(Type=Bone,Name="index_03_r")
    FalloffRegionScaleFactors=(PositiveWidth=1,NegativeWidth=1,PositiveHeight=1,NegativeHeight=1)
    FalloffSize=0.4
    FlipHeightScaling=false
    FlipWidthScaling=true
    OptionalParentItem=(Type=Bone,Name="hand_r")
    RotationOffset=(X=-90,Y=-20,Z=-70)
   )
   > &RigUnit_SetRelativeTransformForItem(
    Child=(Type=Bone,Name="cr_index_03_r")
    Parent=(Type=Bone,Name="hand_r")
    bParentInitial=False
    bPropagateToChildren=True
    Value=&RigUnit_GetRelativeTransformForItem(Child=(Type=Bone,Name="index_03_r"), Parent=(Type=Bone,Name="hand_r"), bChildInitial=False, bParentInitial=False).RelativeTransform
    Weight=@n6.OutputParam
   ) {template="Set Relative Transform::Execute(in Child,in Parent,in bParentInitial,in Value,in Weight,in bPropagateToChildren)"}
   n7 = &RigUnit_ItemArray(Items=((Type=Bone,Name="cr_index_metacarpal_l"),(Type=Bone,Name="cr_index_01_l"),(Type=Bone,Name="cr_index_02_l"),(Type=Bone,Name="cr_index_03_l")))
   SplineFromItems = RigVMFunctionReferenceNode(Samples Per Segment=8, Spline Mode=Hermite, Items=@n7.Items, ExecuteContext=@RigVMFunction_Sequence.C)
   RigUnit_ItemArray = &RigUnit_ItemArray(Items="((Type=Bone,Name=\"deform_tentacle_finger_01_l\"),(Type=Bone,Name=\"deform_tentacle_finger_02_l\"),(Type=Bone,Name=\"deform_tentacle_finger_03_l\"),(Type=Bone,Name=\"deform_tentacle_finger_04_l\"),(Type=Bone,Name=\"deform_tentacle_finger_05_l\"),(Type=Bone,Name=\"def ...(truncated)")
   > /Script/ControlRigSpline.RigUnit_FitChainToSplineCurveItemArray(
    Alignment=Stretched
    DebugSettings=(bEnabled=False,Scale=1,CurveColor=(R=1,G=1,B=0,A=1),SegmentsColor=(R=1,G=0,B=0,A=1),WorldOffset=(Rotation=(X=0,Y=0,Z=0,W=1),Translation=(X=0,Y=0,Z=0),Scale3D=(X=1,Y=1,Z=1)))
    Maximum=1
    Minimum=0
    PoleVectorPosition=(X=0,Y=0,Z=0)
    PrimaryAxis=(X=1,Y=0,Z=0)
    RotationEaseType=Linear
    Rotations=()
    SamplingPrecision=12
    SecondaryAxis=(X=0,Y=0,Z=0)
    Weight=1
    bPropagateToChildren=true
    Items=@RigUnit_ItemArray.Items
    Spline=@SplineFromItems.Spline
   ) {is=struct RigVMUnitNode}
   n8 = &RigUnit_ItemArray(Items=((Type=Bone,Name="cr_index_metacarpal_r"),(Type=Bone,Name="cr_index_01_r"),(Type=Bone,Name="cr_index_02_r"),(Type=Bone,Name="cr_index_03_r")))
   SplineFromItems_1 = RigVMFunctionReferenceNode(Samples Per Segment=8, Spline Mode=Hermite, Items=@n8.Items, ExecuteContext=@RigVMFunction_Sequence.D)
   n9 = &RigUnit_ItemArray(Items="((Type=Bone,Name=\"deform_tentacle_finger_01_r\"),(Type=Bone,Name=\"deform_tentacle_finger_02_r\"),(Type=Bone,Name=\"deform_tentacle_finger_03_r\"),(Type=Bone,Name=\"deform_tentacle_finger_04_r\"),(Type=Bone,Name=\"deform_tentacle_finger_05_r\"),(Type=Bone,Name=\"def ...(truncated)")
   > &RigUnit_FitChainToSplineCurveItemArray(
    Alignment=Stretched
    DebugSettings=(bEnabled=False,Scale=1,CurveColor=(R=1,G=1,B=0,A=1),SegmentsColor=(R=1,G=0,B=0,A=1),WorldOffset=(Rotation=(X=0,Y=0,Z=0,W=1),Translation=(X=0,Y=0,Z=0),Scale3D=(X=1,Y=1,Z=1)))
    Maximum=1
    Minimum=0
    PoleVectorPosition=(X=0,Y=0,Z=0)
    PrimaryAxis=(X=-1,Y=0,Z=0)
    RotationEaseType=Linear
    Rotations=()
    SamplingPrecision=12
    SecondaryAxis=(X=0,Y=0,Z=0)
    Weight=1
    bPropagateToChildren=true
    Items=@n9.Items
    Spline=@SplineFromItems_1.Spline
   )
   RigVMFunctionReturnNode(ExecuteContext=@RigVMFunction_Sequence.E)
   /Script/ControlRigSpline.RigUnit_DrawControlRigSpline(Color=(R=1,G=0,B=0,A=1), Detail=16, ExecutePin=(), Thickness=0.3, Spline=@SplineFromItems.Spline) {is=struct RigVMUnitNode}
   &RigUnit_DrawControlRigSpline(Color=(R=1,G=0,B=0,A=1), Detail=16, ExecutePin=(), Thickness=0.3, Spline=@SplineFromItems_1.Spline)
   RigUnit_GetTransform = &RigUnit_GetTransform(Item=(Type=Bone,Name="cr_index_03_l"), bInitial=False)
   RigVMFunction_DebugTransformArrayMutableNoSpace = /Script/RigVM.RigVMFunction_DebugTransformArrayMutableNoSpace(
    Color=(R=1,G=1,B=1,A=1)
    DebugDrawSettings=(DepthPriority=SDPG_Foreground,Lifetime=-1)
    ExecutePin=()
    Mode=Axes
    ParentIndices=()
    Scale=10
    Thickness=1
    Transforms=((Rotation=(X=0,Y=0,Z=0,W=1),Translation=(X=0,Y=0,Z=0),Scale3D=(X=1,Y=1,Z=1)),(Rotation=(X=0,Y=0,Z=0,W=1),Translation=(X=0,Y=0,Z=0),Scale3D=(X=1,Y=1,Z=1)))
    WorldOffset=(Rotation=(X=0,Y=0,Z=0,W=1),Translation=(X=0,Y=0,Z=0),Scale3D=(X=1,Y=1,Z=1))
    bEnabled=true
    Transforms.0=@RigUnit_GetTransform.Transform
    Transforms.1=&RigUnit_GetTransform(Item=(Type=Bone,Name="cr_index_03_r"), bInitial=False).Transform
   ) {is=struct RigVMUnitNode}
  ]}
 New Function_ContainedGraph {nodes [
   RigVMFunctionEntryNode()
   > RigVMFunction_Sequence = RigVMAggregateNode()
   n1 = &RigUnit_SphericalPoseReader(
    ActiveRegionScaleFactors=(PositiveWidth=1,NegativeWidth=1,PositiveHeight=1,NegativeHeight=1)
    ActiveRegionSize=0.1
    Debug=(bDrawDebug=false,bDraw2D=false,bDrawLocalAxes=false,DebugScale=25,DebugSegments=20,DebugThickness=0.25)
    DriverAxis=(X=1,Y=0,Z=0)
    DriverItem=(Type=Bone,Name="upperarm_l")
    FalloffRegionScaleFactors=(PositiveWidth=1,NegativeWidth=1,PositiveHeight=1,NegativeHeight=1)
    FalloffSize=0.2
    FlipHeightScaling=false
    FlipWidthScaling=false
    OptionalParentItem=(Type=Bone,Name="clavicle_l")
    RotationOffset=(X=185,Y=70,Z=0)
    ExecutePin=@RigVMFunction_Sequence.A
   )
   > n2 = &RigUnit_SphericalPoseReader(
    ActiveRegionScaleFactors=(PositiveWidth=1,NegativeWidth=1,PositiveHeight=1,NegativeHeight=1)
    ActiveRegionSize=0.1
    Debug=(bDrawDebug=false,bDraw2D=false,bDrawLocalAxes=false,DebugScale=25,DebugSegments=20,DebugThickness=0.25)
    DriverAxis=(X=1,Y=0,Z=0)
    DriverItem=(Type=Bone,Name="upperarm_l")
    FalloffRegionScaleFactors=(PositiveWidth=1,NegativeWidth=1,PositiveHeight=1,NegativeHeight=1)
    FalloffSize=0.2
    FlipHeightScaling=false
    FlipWidthScaling=false
    OptionalParentItem=(Type=Bone,Name="clavicle_l")
    RotationOffset=(X=185,Y=130,Z=0)
   )
   > n3 = &RigUnit_SphericalPoseReader(
    ActiveRegionScaleFactors=(PositiveWidth=1,NegativeWidth=1,PositiveHeight=1,NegativeHeight=1)
    ActiveRegionSize=0.15
    Debug=(bDrawDebug=false,bDraw2D=false,bDrawLocalAxes=false,DebugScale=25,DebugSegments=20,DebugThickness=0.25)
    DriverAxis=(X=1,Y=0,Z=0)
    DriverItem=(Type=Bone,Name="upperarm_l")
    FalloffRegionScaleFactors=(PositiveWidth=1,NegativeWidth=1,PositiveHeight=1,NegativeHeight=1)
    FalloffSize=0.35
    FlipHeightScaling=false
    FlipWidthScaling=false
    OptionalParentItem=(Type=Bone,Name="clavicle_l")
    RotationOffset=(X=160,Y=110,Z=0)
   )
   Multiply_2_5 = RigVMAggregateNode(A=@n2.OutputParam, C=RigVMVariableNode() {variable=Is not playing emote Alpha}.Value, B=RigVMVariableNode() {variable=IsRifle}.Value)
   Multiply_2_6 = RigVMAggregateNode(A=@n1.OutputParam, B=RigVMVariableNode() {variable=IsRifle}.Value, C=RigVMVariableNode() {variable=Is not playing emote Alpha}.Value)
   Multiply_2_7 = RigVMAggregateNode(A=@n3.OutputParam, C=RigVMVariableNode() {variable=Is not playing emote Alpha}.Value, B=RigVMVariableNode() {variable=IsRifle}.Value)
   Add_1 = RigVMAggregateNode(B=@Multiply_2_5.Result, A=@Multiply_2_6.Result, C=@Multiply_2_7.Result)
   > n4 = &RigUnit_OffsetTransformForItem(
    Item=(Type=Bone,Name="lowerarm_l_offset")
    OffsetTransform=(Rotation=(X=0,Y=0,Z=0,W=1),Translation=(X=5,Y=0,Z=5),Scale3D=(X=1,Y=1,Z=1))
    bPropagateToChildren=True
    Weight=&RigVMFunction_MathDoubleClamp(Value=@Add_1.Result).Result
   )
   n5 = &RigUnit_GetTransform(Item=(Type=Bone,Name="upperarm_l"), bInitial=False)
   n6 = &RigUnit_ItemArray(Items="((Type=Bone,Name=\"tentacle_01_l\"),(Type=Bone,Name=\"tentacle_02_l\"),(Type=Bone,Name=\"tentacle_03_l\"),(Type=Bone,Name=\"tentacle_04_l\"),(Type=Bone,Name=\"tentacle_05_l\"),(Type=Bone,Name=\"tentacle_06_l\"),(Type=Bone,Name=\"tentacle_07_l\"),(Type=Bone,Name=\"tentacl ...(truncated)")
   AttachChainToSpline = RigVMFunctionReferenceNode(
    Draw Spline=false
    Draw Spline Thickness=1
    Points=((X=0,Y=0,Z=0),(X=0,Y=0,Z=0),(X=0,Y=0,Z=0),(X=0,Y=0,Z=0))
    Pole Vector Position=(X=0,Y=0,Z=0)
    Primary Axis=(X=1,Y=0,Z=0)
    Samples Per Segment=12
    Secondary Axis=(X=0,Y=0,Z=0)
    Spline Mode=Hermite
    Stretch=true
    Execute=@n4.ExecutePin
    Points.0=@n5.Transform.Translation
    Points.1=@n5.Transform.Translation
    Points.2=&RigUnit_GetTransform(Item=(Type=Bone,Name="lowerarm_l_offset"), bInitial=False).Transform.Translation
    Points.3=&RigUnit_GetTransform(Item=(Type=Bone,Name="hand_l"), bInitial=false).Transform.Translation
    Items=@n6.Items
   )
   &RigUnit_ParentConstraint(
    AdvancedSettings=(InterpolationType=Average,RotationOrderForFilter=XZY)
    Child=(Type=Bone,Name="tentacle_10_l")
    Parents=((Item=(Type=Bone,Name="hand_l"),Weight=0.5),(Item=(Type=Bone,Name="tentacle_09_l"),Weight=0.5))
    Weight=1
    ExecutePin=@AttachChainToSpline.Execute
   )
   n7 = &RigUnit_SphericalPoseReader(
    ActiveRegionScaleFactors=(PositiveWidth=1,NegativeWidth=1,PositiveHeight=1,NegativeHeight=1)
    ActiveRegionSize=0.1
    Debug=(bDrawDebug=false,bDraw2D=false,bDrawLocalAxes=false,DebugScale=25,DebugSegments=20,DebugThickness=0.25)
    DriverAxis=(X=1,Y=0,Z=0)
    DriverItem=(Type=Bone,Name="upperarm_l")
    FalloffRegionScaleFactors=(PositiveWidth=1,NegativeWidth=1,PositiveHeight=1,NegativeHeight=1)
    FalloffSize=0.2
    FlipHeightScaling=false
    FlipWidthScaling=false
    OptionalParentItem=(Type=Bone,Name="clavicle_l")
    RotationOffset=(X=185,Y=70,Z=0)
    ExecutePin=@RigVMFunction_Sequence.B
   )
   > n8 = &RigUnit_SphericalPoseReader(
    ActiveRegionScaleFactors=(PositiveWidth=1,NegativeWidth=1,PositiveHeight=1,NegativeHeight=1)
    ActiveRegionSize=0.1
    Debug=(bDrawDebug=false,bDraw2D=false,bDrawLocalAxes=false,DebugScale=25,DebugSegments=20,DebugThickness=0.25)
    DriverAxis=(X=1,Y=0,Z=0)
    DriverItem=(Type=Bone,Name="upperarm_l")
    FalloffRegionScaleFactors=(PositiveWidth=1,NegativeWidth=1,PositiveHeight=1,NegativeHeight=1)
    FalloffSize=0.2
    FlipHeightScaling=false
    FlipWidthScaling=false
    OptionalParentItem=(Type=Bone,Name="clavicle_l")
    RotationOffset=(X=185,Y=130,Z=0)
   )
   > n9 = &RigUnit_SphericalPoseReader(
    ActiveRegionScaleFactors=(PositiveWidth=1,NegativeWidth=1,PositiveHeight=1,NegativeHeight=1)
    ActiveRegionSize=0.15
    Debug=(bDrawDebug=false,bDraw2D=false,bDrawLocalAxes=false,DebugScale=25,DebugSegments=20,DebugThickness=0.25)
    DriverAxis=(X=1,Y=0,Z=0)
    DriverItem=(Type=Bone,Name="upperarm_l")
    FalloffRegionScaleFactors=(PositiveWidth=1,NegativeWidth=1,PositiveHeight=1,NegativeHeight=1)
    FalloffSize=0.35
    FlipHeightScaling=false
    FlipWidthScaling=false
    OptionalParentItem=(Type=Bone,Name="clavicle_l")
    RotationOffset=(X=160,Y=110,Z=0)
   )
   Multiply_2_1 = RigVMAggregateNode(A=@n8.OutputParam, B=RigVMVariableNode() {variable=IsRifle}.Value, C=RigVMVariableNode() {variable=Is not playing emote Alpha}.Value)
   Multiply_2_2 = RigVMAggregateNode(A=@n7.OutputParam, B=RigVMVariableNode() {variable=IsRifle}.Value, C=RigVMVariableNode() {variable=Is not playing emote Alpha}.Value)
   Multiply_2_3 = RigVMAggregateNode(A=@n9.OutputParam, B=RigVMVariableNode() {variable=IsRifle}.Value, C=RigVMVariableNode() {variable=Is not playing emote Alpha}.Value)
   Add = RigVMAggregateNode(B=@Multiply_2_1.Result, A=@Multiply_2_2.Result, C=@Multiply_2_3.Result)
   Clamp = &RigVMFunction_MathDoubleClamp(Value=@Add.Result)
   > &RigUnit_OffsetTransformForItem(
    Item=(Type=Bone,Name="lowerarm_r_offset")
    OffsetTransform=(Rotation=(X=0,Y=0,Z=0,W=1),Translation=(X=-5,Y=0,Z=-5),Scale3D=(X=1,Y=1,Z=1))
    bPropagateToChildren=True
    Weight=@Clamp.Result
   )
   > n10 = &RigUnit_SphericalPoseReader(
    ActiveRegionScaleFactors=(PositiveWidth=1,NegativeWidth=1,PositiveHeight=1,NegativeHeight=1)
    ActiveRegionSize=0.1
    Debug=(bDrawDebug=false,bDraw2D=false,bDrawLocalAxes=false,DebugScale=25,DebugSegments=20,DebugThickness=0.25)
    DriverAxis=(X=-1,Y=0,Z=0)
    DriverItem=(Type=Bone,Name="upperarm_r")
    FalloffRegionScaleFactors=(PositiveWidth=1,NegativeWidth=1,PositiveHeight=1,NegativeHeight=1)
    FalloffSize=0.35
    OptionalParentItem=(Type=Bone,Name="clavicle_r")
    RotationOffset=(X=-145,Y=210,Z=0)
   )
   Multiply_2_4 = RigVMAggregateNode(A=@n10.OutputParam, B=RigVMVariableNode() {variable=IsPistol}.Value, C=RigVMVariableNode() {variable=Is not playing emote Alpha}.Value)
   > n11 = &RigUnit_OffsetTransformForItem(
    Item=(Type=Bone,Name="lowerarm_r_offset")
    OffsetTransform=(Rotation=(X=0,Y=0,Z=0,W=1),Translation=(X=0,Y=0,Z=-5),Scale3D=(X=1,Y=1,Z=1))
    bPropagateToChildren=True
    Weight=@Multiply_2_4.Result
   )
   n12 = &RigUnit_GetTransform(Item=(Type=Bone,Name="upperarm_r"), bInitial=False)
   n13 = &RigUnit_ItemArray(Items="((Type=Bone,Name=\"tentacle_01_r\"),(Type=Bone,Name=\"tentacle_02_r\"),(Type=Bone,Name=\"tentacle_03_r\"),(Type=Bone,Name=\"tentacle_04_r\"),(Type=Bone,Name=\"tentacle_05_r\"),(Type=Bone,Name=\"tentacle_06_r\"),(Type=Bone,Name=\"tentacle_07_r\"),(Type=Bone,Name=\"tentacl ...(truncated)")
   AttachChainToSpline_1 = RigVMFunctionReferenceNode(
    Draw Spline=false
    Draw Spline Thickness=1
    Points=((X=0,Y=0,Z=0),(X=0,Y=0,Z=0),(X=0,Y=0,Z=0),(X=0,Y=0,Z=0))
    Pole Vector Position=(X=0,Y=0,Z=0)
    Primary Axis=(X=-1,Y=0,Z=0)
    Samples Per Segment=12
    Secondary Axis=(X=0,Y=0,Z=0)
    Spline Mode=Hermite
    Stretch=true
    Execute=@n11.ExecutePin
    Points.0=@n12.Transform.Translation
    Points.1=@n12.Transform.Translation
    Points.2=&RigUnit_GetTransform(Item=(Type=Bone,Name="lowerarm_r_offset"), bInitial=False).Transform.Translation
    Points.3=&RigUnit_GetTransform(Item=(Type=Bone,Name="hand_r"), bInitial=false).Transform.Translation
    Items=@n13.Items
   )
   &RigUnit_ParentConstraint(
    AdvancedSettings=(InterpolationType=Average,RotationOrderForFilter=XZY)
    Child=(Type=Bone,Name="tentacle_10_r")
    Parents=((Item=(Type=Bone,Name="hand_r"),Weight=0.5),(Item=(Type=Bone,Name="tentacle_09_r"),Weight=0.5))
    Weight=1
    ExecutePin=@AttachChainToSpline_1.Execute
   )
   RigVMFunctionReturnNode(ExecuteContext=@RigVMFunction_Sequence.C)
  ]}
 New Function_ContainedGraph {nodes [
   RigVMFunctionEntryNode()
   > RigVMFunction_Sequence = RigVMAggregateNode()
   RigVMFunctionReturnNode(ExecuteContext=@RigVMFunction_Sequence.D)
   n1 = &RigUnit_SphericalPoseReader(
    ActiveRegionScaleFactors=(PositiveWidth=0.5,NegativeWidth=0.5,PositiveHeight=1,NegativeHeight=1)
    ActiveRegionSize=0.15
    Debug=(bDrawDebug=false,bDraw2D=false,bDrawLocalAxes=false,DebugScale=25,DebugSegments=20,DebugThickness=0.25)
    DriverAxis=(X=-1,Y=0,Z=0)
    DriverItem=(Type=Bone,Name="upperarm_r")
    FalloffRegionScaleFactors=(PositiveWidth=1,NegativeWidth=0.5,PositiveHeight=1,NegativeHeight=1)
    FalloffSize=0.35
    OptionalParentItem=(Type=Bone,Name="spine_05")
    RotationOffset=(X=270,Y=-45,Z=0)
    ExecutePin=@RigVMFunction_Sequence.B
   )
   > &RigUnit_ModifyTransforms(
    ItemToModify=((Item=(Type=Bone,Name="chest_arms_ft_r_MSTR"),Transform=(Rotation=(X=0,Y=0,Z=0,W=1),Translation=(X=0,Y=0,Z=0),Scale3D=(X=1,Y=1,Z=1))))
    ItemToModify.0.Transform.Rotation=&RigVMFunction_MathQuaternionFromEuler(Euler=(X=0,Y=-20,Z=-20)).Result
    Weight=@n1.OutputParam
   )
   > &RigUnit_ModifyTransforms(
    ItemToModify=((Item=(Type=Bone,Name="chest_arms_ft_r_02_jnt"),Transform=(Rotation=(X=0,Y=0,Z=0,W=1),Translation=(X=0,Y=0,Z=0),Scale3D=(X=1,Y=1,Z=1))))
    ItemToModify.0.Transform.Rotation=&RigVMFunction_MathQuaternionFromEuler(Euler=(X=0,Y=-20,Z=-20)).Result
    Weight=@n1.OutputParam
   )
   > &RigUnit_ModifyTransforms(
    ItemToModify=((Item=(Type=Bone,Name="chest_arms_ft_r_04_jnt"),Transform=(Rotation=(X=0,Y=0,Z=0,W=1),Translation=(X=0,Y=0,Z=0),Scale3D=(X=1,Y=1,Z=1))))
    ItemToModify.0.Transform.Rotation=&RigVMFunction_MathQuaternionFromEuler(Euler=(X=0,Y=-20,Z=-20)).Result
    Weight=@n1.OutputParam
   )
   > &RigUnit_ModifyTransforms(
    ItemToModify=((Item=(Type=Bone,Name="chest_arms_ft_l_MSTR"),Transform=(Rotation=(X=0,Y=0,Z=0,W=1),Translation=(X=3,Y=0,Z=0),Scale3D=(X=1,Y=1,Z=1))))
    ItemToModify.0.Transform.Rotation=&RigVMFunction_MathQuaternionFromEuler(Euler=(X=0,Y=0,Z=0)).Result
    Weight=@n1.OutputParam
   )
   > &RigUnit_ModifyTransforms(
    ItemToModify=((Item=(Type=Bone,Name="chest_arms_ft_l_Offset_MSTR"),Transform=(Rotation=(X=0,Y=0,Z=0,W=1),Translation=(X=3,Y=0,Z=0),Scale3D=(X=1,Y=1,Z=1))))
    ItemToModify.0.Transform.Rotation=&RigVMFunction_MathQuaternionFromEuler(Euler=(X=0,Y=0,Z=0)).Result
    Weight=@n1.OutputParam
   )
   > n2 = &RigUnit_SphericalPoseReader(
    ActiveRegionScaleFactors=(PositiveWidth=1,NegativeWidth=1,PositiveHeight=1,NegativeHeight=1)
    ActiveRegionSize=0.15
    Debug=(bDrawDebug=false,bDraw2D=false,bDrawLocalAxes=false,DebugScale=25,DebugSegments=20,DebugThickness=0.25)
    DriverAxis=(X=-1,Y=0,Z=0)
    DriverItem=(Type=Bone,Name="upperarm_r")
    FalloffRegionScaleFactors=(PositiveWidth=1,NegativeWidth=1,PositiveHeight=1,NegativeHeight=1)
    FalloffSize=0.35
    OptionalParentItem=(Type=Bone,Name="clavicle_r")
    RotationOffset=(X=-140,Y=210,Z=0)
   )
   Multiply_2_3 = RigVMAggregateNode(A=@n2.OutputParam, B=RigVMVariableNode() {variable=IsPistol}.Value, C=RigVMVariableNode() {variable=Is not playing emote Alpha}.Value)
   > &RigUnit_ModifyTransforms(
    ItemToModify=((Item=(Type=Bone,Name="chest_arms_ft_l_Offset_MSTR"),Transform=(Rotation=(X=0,Y=0,Z=0,W=1),Translation=(X=3,Y=-3,Z=0),Scale3D=(X=1,Y=1,Z=1))))
    ItemToModify.0.Transform.Rotation=&RigVMFunction_MathQuaternionFromEuler(Euler=(X=0,Y=0,Z=0)).Result
    Weight=@Multiply_2_3.Result
   )
   &RigUnit_ParentConstraint(
    AdvancedSettings=(InterpolationType=Average,RotationOrderForFilter=XZY)
    Child=(Type=Bone,Name="extra_lowerarm_03pos_l_jnt")
    Parents=((Item=(Type=Bone,Name="extra_lowerarm_l_01_jnt"),Weight=1),(Item=(Type=Bone,Name="extra_lowerarm_06pos_l_jnt"),Weight=1))
    Weight=1
    ExecutePin=@RigVMFunction_Sequence.A
   )
   > &RigUnit_ParentConstraint(
    AdvancedSettings=(InterpolationType=Average,RotationOrderForFilter=XZY)
    Child=(Type=Bone,Name="extra_lowerarm_04pos_l_jnt")
    Parents=((Item=(Type=Bone,Name="extra_lowerarm_l_01_jnt"),Weight=1),(Item=(Type=Bone,Name="extra_lowerarm_06pos_l_jnt"),Weight=1))
    Weight=1
   )
   > &RigUnit_ParentConstraint(
    AdvancedSettings=(InterpolationType=Average,RotationOrderForFilter=XZY)
    Child=(Type=Bone,Name="extra_lowerarm_03pos_r_jnt")
    Parents=((Item=(Type=Bone,Name="extra_lowerarm_r_01_jnt"),Weight=1),(Item=(Type=Bone,Name="extra_lowerarm_05pos_r_jnt"),Weight=1))
    Weight=1
   )
   > &RigUnit_ParentConstraint(
    AdvancedSettings=(InterpolationType=Average,RotationOrderForFilter=XZY)
    Child=(Type=Bone,Name="extra_lowerarm_04pos_r_jnt")
    Parents=((Item=(Type=Bone,Name="extra_lowerarm_r_01_jnt"),Weight=1),(Item=(Type=Bone,Name="extra_lowerarm_05pos_r_jnt"),Weight=1))
    Weight=1
   )
   > &RigUnit_ParentConstraint(
    AdvancedSettings=(InterpolationType=Average,RotationOrderForFilter=XZY)
    Child=(Type=Bone,Name="extra_lowerarm_06pos_l_jnt")
    Parents=((Item=(Type=Bone,Name="Handle_aim_jnt"),Weight=1))
    Weight=1
   )
   > n3 = &RigUnit_SphericalPoseReader(
    ActiveRegionScaleFactors=(PositiveWidth=1,NegativeWidth=1,PositiveHeight=1,NegativeHeight=1)
    ActiveRegionSize=0.15
    Debug=(bDrawDebug=false,bDraw2D=false,bDrawLocalAxes=false,DebugScale=25,DebugSegments=20,DebugThickness=0.25)
    DriverAxis=(X=-1,Y=0,Z=0)
    DriverItem=(Type=Bone,Name="upperarm_r")
    FalloffRegionScaleFactors=(PositiveWidth=1,NegativeWidth=1,PositiveHeight=1,NegativeHeight=1)
    FalloffSize=0.35
    OptionalParentItem=(Type=Bone,Name="clavicle_r")
    RotationOffset=(X=-105,Y=135,Z=0)
   )
   > &RigUnit_ModifyTransforms(
    ItemToModify=((Item=(Type=Bone,Name="extra_lowerarm_03pos_l_jnt"),Transform=(Rotation=(X=0,Y=0,Z=0,W=1),Translation=(X=0,Y=0,Z=2),Scale3D=(X=1,Y=1,Z=1))))
    ItemToModify.0.Transform.Rotation=&RigVMFunction_MathQuaternionFromEuler(Euler=(X=0,Y=0,Z=0)).Result
    Weight=@n3.OutputParam
   )
   > n4 = &RigUnit_ModifyTransforms(
    ItemToModify=((Item=(Type=Bone,Name="extra_lowerarm_04pos_l_jnt"),Transform=(Rotation=(X=0,Y=0,Z=0,W=1),Translation=(X=0,Y=0,Z=2),Scale3D=(X=1,Y=1,Z=1))))
    ItemToModify.0.Transform.Rotation=&RigVMFunction_MathQuaternionFromEuler(Euler=(X=0,Y=0,Z=0)).Result
    Weight=@n3.OutputParam
   )
   n5 = &RigUnit_ItemArray(Items="((Type=Bone,Name=\"extra_lowerarm_l_01_jnt\"),(Type=Bone,Name=\"extra_lowerarm_l_02_jnt\"),(Type=Bone,Name=\"extra_lowerarm_l_03_jnt\"),(Type=Bone,Name=\"extra_lowerarm_l_04_jnt\"),(Type=Bone,Name=\"extra_lowerarm_l_05_jnt\"),(Type=Bone,Name=\"extra_lowerarm_l_06_jnt ...(truncated)")
   AttachChainToSpline = RigVMFunctionReferenceNode(
    Draw Spline=false
    Draw Spline Thickness=1
    Points=((X=0,Y=0,Z=0),(X=0,Y=0,Z=0),(X=0,Y=0,Z=0),(X=0,Y=0,Z=0))
    Pole Vector Position=(X=0,Y=0,Z=0)
    Primary Axis=(X=1,Y=0,Z=0)
    Samples Per Segment=16
    Secondary Axis=(X=0,Y=0,Z=0)
    Spline Mode=Hermite
    Stretch=true
    Execute=@n4.ExecutePin
    Points.1=&RigUnit_GetTransform(Item=(Type=Bone,Name="extra_lowerarm_03pos_l_jnt"), bInitial=False).Transform.Translation
    Points.2=&RigUnit_GetTransform(Item=(Type=Bone,Name="extra_lowerarm_04pos_l_jnt"), bInitial=false).Transform.Translation
    Points.0=&RigUnit_GetTransform(Item=(Type=Bone,Name="extra_lowerarm_l_01_jnt"), bInitial=False).Transform.Translation
    Points.3=&RigUnit_GetTransform(Item=(Type=Bone,Name="extra_lowerarm_06pos_l_jnt"), bInitial=false).Transform.Translation
    Items=@n5.Items
   )
   &RigUnit_ModifyTransforms(
    ItemToModify=((Item=(Type=Bone,Name="extra_lowerarm_03pos_r_jnt"),Transform=(Rotation=(X=0,Y=0,Z=0,W=1),Translation=(X=0,Y=0,Z=2),Scale3D=(X=1,Y=1,Z=1))))
    ExecutePin=@AttachChainToSpline.Execute
    ItemToModify.0.Transform.Rotation=&RigVMFunction_MathQuaternionFromEuler(Euler=(X=0,Y=0,Z=0)).Result
    Weight=@n3.OutputParam
   )
   > n6 = &RigUnit_ModifyTransforms(
    ItemToModify=((Item=(Type=Bone,Name="extra_lowerarm_04pos_r_jnt"),Transform=(Rotation=(X=0,Y=0,Z=0,W=1),Translation=(X=0,Y=0,Z=2),Scale3D=(X=1,Y=1,Z=1))))
    ItemToModify.0.Transform.Rotation=&RigVMFunction_MathQuaternionFromEuler(Euler=(X=0,Y=0,Z=0)).Result
    Weight=@n3.OutputParam
   )
   n7 = &RigUnit_ItemArray(Items=((Type=Bone,Name="extra_lowerarm_r_01_jnt"),(Type=Bone,Name="extra_lowerarm_r_02_jnt"),(Type=Bone,Name="extra_lowerarm_r_03_jnt"),(Type=Bone,Name="extra_lowerarm_r_04_jnt"),(Type=Bone,Name="extra_lowerarm_r_05_jnt")))
   AttachChainToSpline_1 = RigVMFunctionReferenceNode(
    Draw Spline=false
    Draw Spline Thickness=1
    Points=((X=0,Y=0,Z=0),(X=0,Y=0,Z=0),(X=0,Y=0,Z=0),(X=0,Y=0,Z=0))
    Pole Vector Position=(X=0,Y=0,Z=0)
    Primary Axis=(X=1,Y=0,Z=0)
    Samples Per Segment=16
    Secondary Axis=(X=0,Y=0,Z=0)
    Spline Mode=Hermite
    Stretch=true
    Execute=@n6.ExecutePin
    Points.1=&RigUnit_GetTransform(Item=(Type=Bone,Name="extra_lowerarm_03pos_r_jnt"), bInitial=False).Transform.Translation
    Points.0=&RigUnit_GetTransform(Item=(Type=Bone,Name="extra_lowerarm_r_01_jnt"), bInitial=False).Transform.Translation
    Points.3=&RigUnit_GetTransform(Item=(Type=Bone,Name="extra_lowerarm_05pos_r_jnt"), bInitial=false).Transform.Translation
    Points.2=&RigUnit_GetTransform(Item=(Type=Bone,Name="extra_lowerarm_04pos_r_jnt"), bInitial=false).Transform.Translation
    Items=@n7.Items
   )
   n8 = &RigUnit_SphericalPoseReader(
    ActiveRegionScaleFactors=(PositiveWidth=1,NegativeWidth=1,PositiveHeight=1,NegativeHeight=1)
    ActiveRegionSize=0.1
    Debug=(bDrawDebug=false,bDraw2D=false,bDrawLocalAxes=false,DebugScale=25,DebugSegments=20,DebugThickness=0.25)
    DriverAxis=(X=-1,Y=0,Z=0)
    DriverItem=(Type=Bone,Name="upperarm_r")
    FalloffRegionScaleFactors=(PositiveWidth=1,NegativeWidth=1,PositiveHeight=1,NegativeHeight=1)
    FalloffSize=0.35
    OptionalParentItem=(Type=Bone,Name="clavicle_r")
    RotationOffset=(X=-145,Y=210,Z=0)
    ExecutePin=@RigVMFunction_Sequence.C
   )
   Multiply_2_1 = RigVMAggregateNode(A=@n8.OutputParam, B=RigVMVariableNode() {variable=IsPistol}.Value, C=RigVMVariableNode() {variable=Is not playing emote Alpha}.Value)
   > &RigUnit_ModifyTransforms(
    ItemToModify=((Item=(Type=Bone,Name="extra_lowerarm_l_01_jnt"),Transform=(Rotation=(X=0,Y=0,Z=0,W=1),Translation=(X=0,Y=0,Z=0),Scale3D=(X=1,Y=1,Z=1))))
    ItemToModify.0.Transform.Rotation=&RigVMFunction_MathQuaternionFromEuler(Euler=(X=0,Y=0,Z=-10)).Result
    Weight=@Multiply_2_1.Result
   )
   > &RigUnit_ModifyTransforms(
    ItemToModify=((Item=(Type=Bone,Name="extra_lowerarm_l_02_jnt"),Transform=(Rotation=(X=0,Y=0,Z=0,W=1),Translation=(X=0,Y=0,Z=0),Scale3D=(X=1,Y=1,Z=1))))
    ItemToModify.0.Transform.Rotation=&RigVMFunction_MathQuaternionFromEuler(Euler=(X=0,Y=5,Z=-10)).Result
    Weight=@Multiply_2_1.Result
   )
   > &RigUnit_ModifyTransforms(
    ItemToModify=((Item=(Type=Bone,Name="extra_lowerarm_l_03_jnt"),Transform=(Rotation=(X=0,Y=0,Z=0,W=1),Translation=(X=0,Y=0,Z=0),Scale3D=(X=1,Y=1,Z=1))))
    ItemToModify.0.Transform.Rotation=&RigVMFunction_MathQuaternionFromEuler(Euler=(X=0,Y=5,Z=-10)).Result
    Weight=@Multiply_2_1.Result
   )
   > &RigUnit_ModifyTransforms(
    ItemToModify=((Item=(Type=Bone,Name="extra_lowerarm_l_04_jnt"),Transform=(Rotation=(X=0,Y=0,Z=0,W=1),Translation=(X=0,Y=0,Z=0),Scale3D=(X=1,Y=1,Z=1))))
    ItemToModify.0.Transform.Rotation=&RigVMFunction_MathQuaternionFromEuler(Euler=(X=0,Y=5,Z=-10)).Result
    Weight=@Multiply_2_1.Result
   )
   > &RigUnit_ModifyTransforms(
    ItemToModify=((Item=(Type=Bone,Name="extra_lowerarm_r_01_jnt"),Transform=(Rotation=(X=0,Y=0,Z=0,W=1),Translation=(X=0,Y=0,Z=0),Scale3D=(X=1,Y=1,Z=1))))
    ItemToModify.0.Transform.Rotation=&RigVMFunction_MathQuaternionFromEuler(Euler=(X=-10,Y=10,Z=10)).Result
    Weight=@Multiply_2_1.Result
   )
   > &RigUnit_ModifyTransforms(
    ItemToModify=((Item=(Type=Bone,Name="extra_lowerarm_r_02_jnt"),Transform=(Rotation=(X=0,Y=0,Z=0,W=1),Translation=(X=0,Y=0,Z=0),Scale3D=(X=1,Y=1,Z=1))))
    ItemToModify.0.Transform.Rotation=&RigVMFunction_MathQuaternionFromEuler(Euler=(X=0,Y=-10,Z=5)).Result
    Weight=@Multiply_2_1.Result
   )
   > &RigUnit_ModifyTransforms(
    ItemToModify=((Item=(Type=Bone,Name="extra_lowerarm_r_03_jnt"),Transform=(Rotation=(X=0,Y=0,Z=0,W=1),Translation=(X=0,Y=0,Z=0),Scale3D=(X=1,Y=1,Z=1))))
    ItemToModify.0.Transform.Rotation=&RigVMFunction_MathQuaternionFromEuler(Euler=(X=0,Y=-5,Z=5)).Result
    Weight=@Multiply_2_1.Result
   )
   > &RigUnit_ModifyTransforms(
    ItemToModify=((Item=(Type=Bone,Name="extra_lowerarm_r_04_jnt"),Transform=(Rotation=(X=0,Y=0,Z=0,W=1),Translation=(X=0,Y=0,Z=0),Scale3D=(X=1,Y=1,Z=1))))
    ItemToModify.0.Transform.Rotation=&RigVMFunction_MathQuaternionFromEuler(Euler=(X=0,Y=-5,Z=5)).Result
    Weight=@Multiply_2_1.Result
   )
   > n9 = &RigUnit_SphericalPoseReader(
    ActiveRegionScaleFactors=(PositiveWidth=1,NegativeWidth=1,PositiveHeight=1,NegativeHeight=1)
    ActiveRegionSize=0.1
    Debug=(bDrawDebug=false,bDraw2D=false,bDrawLocalAxes=false,DebugScale=25,DebugSegments=20,DebugThickness=0.25)
    DriverAxis=(X=1,Y=0,Z=0)
    DriverItem=(Type=Bone,Name="upperarm_l")
    FalloffRegionScaleFactors=(PositiveWidth=1,NegativeWidth=1,PositiveHeight=1,NegativeHeight=1)
    FalloffSize=0.2
    FlipHeightScaling=false
    FlipWidthScaling=false
    OptionalParentItem=(Type=Bone,Name="clavicle_l")
    RotationOffset=(X=185,Y=70,Z=0)
   )
   Multiply_2_2 = RigVMAggregateNode(A=@n9.OutputParam, B=RigVMVariableNode() {variable=IsRifle}.Value, C=RigVMVariableNode() {variable=Is not playing emote Alpha}.Value)
   > &RigUnit_ModifyTransforms(
    ItemToModify=((Item=(Type=Bone,Name="extra_lowerarm_l_01_jnt"),Transform=(Rotation=(X=0,Y=0,Z=0,W=1),Translation=(X=0,Y=0,Z=0),Scale3D=(X=1,Y=1,Z=1))))
    ItemToModify.0.Transform.Rotation=&RigVMFunction_MathQuaternionFromEuler(Euler=(X=0,Y=-5,Z=-5)).Result
    Weight=@Multiply_2_2.Result
   )
   > &RigUnit_ModifyTransforms(
    ItemToModify=((Item=(Type=Bone,Name="extra_lowerarm_l_02_jnt"),Transform=(Rotation=(X=0,Y=0,Z=0,W=1),Translation=(X=0,Y=0,Z=0),Scale3D=(X=1,Y=1,Z=1))))
    ItemToModify.0.Transform.Rotation=&RigVMFunction_MathQuaternionFromEuler(Euler=(X=0,Y=-5,Z=-5)).Result
    Weight=@Multiply_2_2.Result
   )
   > &RigUnit_ModifyTransforms(
    ItemToModify=((Item=(Type=Bone,Name="extra_lowerarm_l_03_jnt"),Transform=(Rotation=(X=0,Y=0,Z=0,W=1),Translation=(X=0,Y=0,Z=0),Scale3D=(X=1,Y=1,Z=1))))
    ItemToModify.0.Transform.Rotation=&RigVMFunction_MathQuaternionFromEuler(Euler=(X=0,Y=-5,Z=-5)).Result
    Weight=@Multiply_2_2.Result
   )
   > &RigUnit_ModifyTransforms(
    ItemToModify=((Item=(Type=Bone,Name="extra_lowerarm_l_04_jnt"),Transform=(Rotation=(X=0,Y=0,Z=0,W=1),Translation=(X=0,Y=0,Z=0),Scale3D=(X=1,Y=1,Z=1))))
    ItemToModify.0.Transform.Rotation=&RigVMFunction_MathQuaternionFromEuler(Euler=(X=0,Y=-5,Z=-5)).Result
    Weight=@Multiply_2_2.Result
   )
   > &RigUnit_ModifyTransforms(
    ItemToModify=((Item=(Type=Bone,Name="chest_arms_ft_l_Offset_MSTR"),Transform=(Rotation=(X=0,Y=0,Z=0,W=1),Translation=(X=3,Y=-3,Z=0),Scale3D=(X=1,Y=1,Z=1))))
    ItemToModify.0.Transform.Rotation=&RigVMFunction_MathQuaternionFromEuler(Euler=(X=0,Y=0,Z=0)).Result
    Weight=@Multiply_2_2.Result
   )
   > &RigUnit_ModifyTransforms(
    ItemToModify=((Item=(Type=Bone,Name="chest_arms_ft_jnt"),Transform=(Rotation=(X=0,Y=0,Z=0,W=1),Translation=(X=-5,Y=0,Z=0),Scale3D=(X=1,Y=1,Z=1))))
    ItemToModify.0.Transform.Rotation=&RigVMFunction_MathQuaternionFromEuler(Euler=(X=5,Y=0,Z=10)).Result
    Weight=@Multiply_2_2.Result
   )
   n10 = &RigUnit_SphericalPoseReader(
    ActiveRegionScaleFactors=(PositiveWidth=1,NegativeWidth=1,PositiveHeight=1,NegativeHeight=1)
    ActiveRegionSize=0.1
    Debug=(bDrawDebug=false,bDraw2D=false,bDrawLocalAxes=false,DebugScale=25,DebugSegments=20,DebugThickness=0.25)
    DriverAxis=(X=-1,Y=0,Z=0)
    DriverItem=(Type=Bone,Name="upperarm_r")
    FalloffRegionScaleFactors=(PositiveWidth=1,NegativeWidth=1,PositiveHeight=1,NegativeHeight=1)
    FalloffSize=0.25
    OptionalParentItem=(Type=Bone,Name="clavicle_r")
    RotationOffset=(X=-145,Y=230,Z=0)
    ExecutePin=@RigVMFunction_Sequence.F
   )
   Multiply_2_4 = RigVMAggregateNode(A=@n10.OutputParam, B=RigVMVariableNode() {variable=IsPistol}.Value, C=RigVMVariableNode() {variable=Is not playing emote Alpha}.Value)
   > &RigUnit_ModifyTransforms(
    ItemToModify=((Item=(Type=Bone,Name="extra_lowerarm_r_01_jnt"),Transform=(Rotation=(X=0,Y=0,Z=0,W=1),Translation=(X=0,Y=0,Z=0),Scale3D=(X=1,Y=1,Z=1))))
    ItemToModify.0.Transform.Rotation=&RigVMFunction_MathQuaternionFromEuler(Euler=(X=0,Y=10,Z=0), RotationOrder=XYZ).Result
    Weight=@Multiply_2_4.Result
   )
   > &RigUnit_ModifyTransforms(
    ItemToModify=((Item=(Type=Bone,Name="extra_lowerarm_r_02_jnt"),Transform=(Rotation=(X=0,Y=0,Z=0,W=1),Translation=(X=0,Y=0,Z=0),Scale3D=(X=1,Y=1,Z=1))))
    ItemToModify.0.Transform.Rotation=&RigVMFunction_MathQuaternionFromEuler(Euler=(X=0,Y=0,Z=0), RotationOrder=XYZ).Result
    Weight=@Multiply_2_4.Result
   )
   > &RigUnit_ModifyTransforms(
    ItemToModify=((Item=(Type=Bone,Name="extra_lowerarm_r_03_jnt"),Transform=(Rotation=(X=0,Y=0,Z=0,W=1),Translation=(X=0,Y=0,Z=0),Scale3D=(X=1,Y=1,Z=1))))
    ItemToModify.0.Transform.Rotation=&RigVMFunction_MathQuaternionFromEuler(Euler=(X=0,Y=0,Z=0), RotationOrder=XYZ).Result
    Weight=@Multiply_2_4.Result
   )
   > &RigUnit_ModifyTransforms(
    ItemToModify=((Item=(Type=Bone,Name="extra_lowerarm_r_04_jnt"),Transform=(Rotation=(X=0,Y=0,Z=0,W=1),Translation=(X=0,Y=0,Z=0),Scale3D=(X=1,Y=1,Z=1))))
    ItemToModify.0.Transform.Rotation=&RigVMFunction_MathQuaternionFromEuler(Euler=(X=0,Y=0,Z=0), RotationOrder=XYZ).Result
    Weight=@Multiply_2_4.Result
   )
   > n11 = &RigUnit_SphericalPoseReader(
    ActiveRegionScaleFactors=(PositiveWidth=1,NegativeWidth=1,PositiveHeight=1,NegativeHeight=1)
    ActiveRegionSize=0.1
    Debug=(bDrawDebug=false,bDraw2D=false,bDrawLocalAxes=false,DebugScale=25,DebugSegments=20,DebugThickness=0.25)
    DriverAxis=(X=1,Y=0,Z=0)
    DriverItem=(Type=Bone,Name="upperarm_l")
    FalloffRegionScaleFactors=(PositiveWidth=1,NegativeWidth=1,PositiveHeight=1,NegativeHeight=1)
    FalloffSize=0.25
    FlipHeightScaling=false
    FlipWidthScaling=false
    OptionalParentItem=(Type=Bone,Name="clavicle_l")
    RotationOffset=(X=155,Y=60,Z=0)
   )
   Multiply_2_5 = RigVMAggregateNode(A=@n11.OutputParam, B=RigVMVariableNode() {variable=IsRifle}.Value, C=RigVMVariableNode() {variable=Is not playing emote Alpha}.Value)
   > &RigUnit_ModifyTransforms(
    ItemToModify=((Item=(Type=Bone,Name="extra_lowerarm_l_01_jnt"),Transform=(Rotation=(X=0,Y=0,Z=0,W=1),Translation=(X=0,Y=0,Z=0),Scale3D=(X=1,Y=1,Z=1))))
    ItemToModify.0.Transform.Rotation=&RigVMFunction_MathQuaternionFromEuler(Euler=(X=0,Y=10,Z=-10), RotationOrder=XYZ).Result
    Weight=@Multiply_2_5.Result
   )
   > &RigUnit_ModifyTransforms(
    ItemToModify=((Item=(Type=Bone,Name="extra_lowerarm_l_02_jnt"),Transform=(Rotation=(X=0,Y=0,Z=0,W=1),Translation=(X=0,Y=0,Z=0),Scale3D=(X=1,Y=1,Z=1))))
    ItemToModify.0.Transform.Rotation=&RigVMFunction_MathQuaternionFromEuler(Euler=(X=0,Y=10,Z=-10), RotationOrder=XYZ).Result
    Weight=@Multiply_2_5.Result
   )
   > &RigUnit_ModifyTransforms(
    ItemToModify=((Item=(Type=Bone,Name="extra_lowerarm_l_03_jnt"),Transform=(Rotation=(X=0,Y=0,Z=0,W=1),Translation=(X=0,Y=0,Z=0),Scale3D=(X=1,Y=1,Z=1))))
    ItemToModify.0.Transform.Rotation=&RigVMFunction_MathQuaternionFromEuler(Euler=(X=0,Y=0,Z=0), RotationOrder=XYZ).Result
    Weight=@Multiply_2_5.Result
   )
   > &RigUnit_ModifyTransforms(
    ItemToModify=((Item=(Type=Bone,Name="extra_lowerarm_l_04_jnt"),Transform=(Rotation=(X=0,Y=0,Z=0,W=1),Translation=(X=0,Y=0,Z=0),Scale3D=(X=1,Y=1,Z=1))))
    ItemToModify.0.Transform.Rotation=&RigVMFunction_MathQuaternionFromEuler(Euler=(X=0,Y=0,Z=0), RotationOrder=XYZ).Result
    Weight=@Multiply_2_5.Result
   )
   > n12 = &RigUnit_SphericalPoseReader(
    ActiveRegionScaleFactors=(PositiveWidth=3.5,NegativeWidth=0.1,PositiveHeight=3,NegativeHeight=3)
    ActiveRegionSize=0.1
    Debug=(bDrawDebug=false,bDraw2D=false,bDrawLocalAxes=false,DebugScale=25,DebugSegments=20,DebugThickness=0.25)
    DriverAxis=(X=-1,Y=0,Z=0)
    DriverItem=(Type=Bone,Name="thigh_l")
    FalloffRegionScaleFactors=(PositiveWidth=1,NegativeWidth=1.1,PositiveHeight=1,NegativeHeight=1)
    FalloffSize=0.45
    OptionalParentItem=(Type=Bone,Name="pelvis")
    RotationOffset=(X=120,Y=0,Z=-17)
   )
   > &RigUnit_ModifyTransforms(
    ItemToModify=((Item=(Type=Bone,Name="pelvis_offset"),Transform=(Rotation=(X=-0.0038,Y=-0.0871,Z=-0.0435,W=0.9952),Translation=(X=2,Y=0,Z=0),Scale3D=(X=1,Y=1,Z=1))))
    WeightMinimum=0
    Weight=@n12.OutputParam
   )
   > n13 = &RigUnit_SphericalPoseReader(
    ActiveRegionScaleFactors=(PositiveWidth=3.5,NegativeWidth=0.1,PositiveHeight=3,NegativeHeight=3)
    ActiveRegionSize=0.1
    Debug=(bDrawDebug=false,bDraw2D=false,DebugScale=25,DebugSegments=20,DebugThickness=0.25)
    DriverAxis=(X=1,Y=0,Z=0)
    DriverItem=(Type=Bone,Name="thigh_r")
    FalloffRegionScaleFactors=(PositiveWidth=1,NegativeWidth=1.1,PositiveHeight=1,NegativeHeight=1)
    FalloffSize=0.45
    OptionalParentItem=(Type=Bone,Name="pelvis")
    RotationOffset=(X=-120,Y=180,Z=-17)
   )
   > &RigUnit_ModifyTransforms(
    ItemToModify=((Item=(Type=Bone,Name="pelvis_offset"),Transform=(Rotation=(X=0.0038,Y=0.0871,Z=-0.0435,W=0.9952),Translation=(X=2,Y=0,Z=0),Scale3D=(X=1,Y=1,Z=1))))
    WeightMinimum=0
    Weight=@n13.OutputParam
   )
   > /Script/ControlRig.RigUnit_ChainHarmonicsPerItem(
    ChainRoot=(Type=Bone,Name="extra_lowerarm_l_04_jnt")
    DrawWorldOffset=(Rotation=(X=0,Y=0,Z=0,W=1),Translation=(X=0,Y=0,Z=0),Scale3D=(X=1,Y=1,Z=1))
    Pendulum=(bEnabled=false,PendulumStiffness=2,PendulumGravity=(X=0,Y=0,Z=0),PendulumBlend=0.75,PendulumDrag=0.98,PendulumMinimum=0,PendulumMaximum=0.1,PendulumEase=Linear,UnwindAxis=(X=0,Y=1,Z=0),UnwindMinimum=0.2,UnwindMaximum=0.05)
    Reach=(bEnabled=false,ReachTarget=(X=0,Y=0,Z=0),ReachAxis=(X=1,Y=0,Z=0),ReachMinimum=0,ReachMaximum=0,ReachEase=Linear)
    Speed=(X=0.5,Y=2,Z=1.5)
    Wave=(bEnabled=true,WaveFrequency=(X=0,Y=3,Z=1),WaveAmplitude=(X=0.1,Y=0.5,Z=0.1),WaveOffset=(X=0,Y=2,Z=0.2),WaveNoise=(X=0,Y=0,Z=0),WaveMinimum=0,WaveMaximum=1,WaveEase=Linear)
    WaveCurve=(EditorCurveData=(Keys=((),(Time=1,Value=1))))
    bDrawDebug=false
   ) {is=struct RigVMUnitNode}
   > &RigUnit_ChainHarmonicsPerItem(
    ChainRoot=(Type=Bone,Name="extra_lowerarm_r_01_jnt")
    DrawWorldOffset=(Rotation=(X=0,Y=0,Z=0,W=1),Translation=(X=0,Y=0,Z=0),Scale3D=(X=1,Y=1,Z=1))
    Pendulum=(bEnabled=false,PendulumStiffness=2,PendulumGravity=(X=0,Y=0,Z=0),PendulumBlend=0.75,PendulumDrag=0.98,PendulumMinimum=0,PendulumMaximum=0.1,PendulumEase=Linear,UnwindAxis=(X=0,Y=1,Z=0),UnwindMinimum=0.2,UnwindMaximum=0.05)
    Reach=(bEnabled=false,ReachTarget=(X=0,Y=0,Z=0),ReachAxis=(X=1,Y=0,Z=0),ReachMinimum=0,ReachMaximum=0,ReachEase=Linear)
    Speed=(X=1,Y=-3,Z=-0.5)
    Wave=(bEnabled=true,WaveFrequency=(X=0,Y=5,Z=1),WaveAmplitude=(X=0.1,Y=-0.5,Z=-0.3),WaveOffset=(X=0.3,Y=0.3,Z=0.5),WaveNoise=(X=0,Y=0,Z=0),WaveMinimum=0,WaveMaximum=1,WaveEase=Linear)
    WaveCurve=(EditorCurveData=(Keys=((),(Time=0.7494,Value=1),(Time=0.9938))))
    bDrawDebug=false
   )
   > &RigUnit_ParentConstraint(
    AdvancedSettings=(InterpolationType=Average,RotationOrderForFilter=XZY)
    Child=(Type=Bone,Name="dyn_gourd_handle_jnt")
    Parents=((Item=(Type=Bone,Name="extra_lowerarm_r_06_jnt"),Weight=1))
    Weight=1
    bMaintainOffset=true
   )
   > &RigUnit_ParentConstraint(
    AdvancedSettings=(InterpolationType=Average,RotationOrderForFilter=XZY)
    Child=(Type=Bone,Name="dyn_gourd_Base_jnt")
    Parents=((Item=(Type=Bone,Name="extra_lowerarm_r_06_jnt"),Weight=1))
    Weight=1
   )
   > &RigUnit_ParentConstraint(
    AdvancedSettings=(InterpolationType=Average,RotationOrderForFilter=XZY)
    Child=(Type=Bone,Name="dyn_gourd_jnt")
    Parents=((Item=(Type=Bone,Name="extra_lowerarm_r_06_jnt"),Weight=1))
    Weight=1
   )
   n14 = &RigUnit_SphericalPoseReader(
    ActiveRegionScaleFactors=(PositiveWidth=3.5,NegativeWidth=1.5,PositiveHeight=2,NegativeHeight=2)
    ActiveRegionSize=0.1
    Debug=(bDrawDebug=false,bDraw2D=false,DebugScale=25,DebugSegments=20,DebugThickness=0.25)
    DriverAxis=(X=-1,Y=0,Z=0)
    DriverItem=(Type=Bone,Name="thigh_l")
    FalloffRegionScaleFactors=(PositiveWidth=1,NegativeWidth=1,PositiveHeight=1,NegativeHeight=1)
    FalloffSize=0.45
    OptionalParentItem=(Type=Bone,Name="pelvis")
    RotationOffset=(X=150,Y=0,Z=0)
    ExecutePin=@RigVMFunction_Sequence.E
   )
   > &RigUnit_ModifyTransforms(
    ItemToModify=((Item=(Type=Bone,Name="extra_lowerarm_l_01_jnt"),Transform=(Rotation=(X=0,Y=0,Z=0,W=1),Translation=(X=0,Y=0,Z=0),Scale3D=(X=1,Y=1,Z=1))))
    ItemToModify.0.Transform.Rotation=&RigVMFunction_MathQuaternionFromEuler(Euler=(X=-10,Y=-10,Z=-10), RotationOrder=XYZ).Result
    Weight=@n14.OutputParam
   )
   > &RigUnit_ModifyTransforms(
    ItemToModify=((Item=(Type=Bone,Name="extra_lowerarm_l_02_jnt"),Transform=(Rotation=(X=0,Y=0,Z=0,W=1),Translation=(X=0,Y=0,Z=0),Scale3D=(X=1,Y=1,Z=1))))
    ItemToModify.0.Transform.Rotation=&RigVMFunction_MathQuaternionFromEuler(Euler=(X=-5,Y=-5,Z=-5), RotationOrder=XYZ).Result
    Weight=@n14.OutputParam
   )
   > &RigUnit_ModifyTransforms(
    ItemToModify=((Item=(Type=Bone,Name="extra_lowerarm_l_03_jnt"),Transform=(Rotation=(X=0,Y=0,Z=0,W=1),Translation=(X=0,Y=0,Z=0),Scale3D=(X=1,Y=1,Z=1))))
    ItemToModify.0.Transform.Rotation=&RigVMFunction_MathQuaternionFromEuler(Euler=(X=5,Y=-5,Z=-5), RotationOrder=XYZ).Result
    Weight=@n14.OutputParam
   )
   > &RigUnit_ModifyTransforms(
    ItemToModify=((Item=(Type=Bone,Name="extra_lowerarm_l_04_jnt"),Transform=(Rotation=(X=0,Y=0,Z=0,W=1),Translation=(X=0,Y=0,Z=0),Scale3D=(X=1,Y=1,Z=1))))
    ItemToModify.0.Transform.Rotation=&RigVMFunction_MathQuaternionFromEuler(Euler=(X=5,Y=-5,Z=-5), RotationOrder=XYZ).Result
    Weight=@n14.OutputParam
   )
   > n15 = &RigUnit_SphericalPoseReader(
    ActiveRegionScaleFactors=(PositiveWidth=3.5,NegativeWidth=1.5,PositiveHeight=2,NegativeHeight=2)
    ActiveRegionSize=0.1
    Debug=(bDrawDebug=false,bDraw2D=false,DebugScale=25,DebugSegments=20,DebugThickness=0.25)
    DriverAxis=(X=1,Y=0,Z=0)
    DriverItem=(Type=Bone,Name="thigh_r")
    FalloffRegionScaleFactors=(PositiveWidth=1,NegativeWidth=1,PositiveHeight=1,NegativeHeight=1)
    FalloffSize=0.45
    OptionalParentItem=(Type=Bone,Name="pelvis")
    RotationOffset=(X=-150,Y=180,Z=0)
   )
   > &RigUnit_ModifyTransforms(
    ItemToModify=((Item=(Type=Bone,Name="extra_lowerarm_r_01_jnt"),Transform=(Rotation=(X=0,Y=0,Z=0,W=1),Translation=(X=0,Y=0,Z=0),Scale3D=(X=1,Y=1,Z=1))))
    ItemToModify.0.Transform.Rotation=&RigVMFunction_MathQuaternionFromEuler(Euler=(X=-10,Y=5,Z=15), RotationOrder=XYZ).Result
    Weight=@n15.OutputParam
   )
   > &RigUnit_ModifyTransforms(
    ItemToModify=((Item=(Type=Bone,Name="extra_lowerarm_r_02_jnt"),Transform=(Rotation=(X=0,Y=0,Z=0,W=1),Translation=(X=1,Y=0,Z=0),Scale3D=(X=1,Y=1,Z=1))))
    ItemToModify.0.Transform.Rotation=&RigVMFunction_MathQuaternionFromEuler(Euler=(X=-5,Y=5,Z=15), RotationOrder=XYZ).Result
    Weight=@n15.OutputParam
   )
   > &RigUnit_ModifyTransforms(
    ItemToModify=((Item=(Type=Bone,Name="extra_lowerarm_r_03_jnt"),Transform=(Rotation=(X=0,Y=0,Z=0,W=1),Translation=(X=1,Y=0,Z=0),Scale3D=(X=1,Y=1,Z=1))))
    ItemToModify.0.Transform.Rotation=&RigVMFunction_MathQuaternionFromEuler(Euler=(X=5,Y=5,Z=10), RotationOrder=XYZ).Result
    Weight=@n15.OutputParam
   )
   > &RigUnit_ModifyTransforms(
    ItemToModify=((Item=(Type=Bone,Name="extra_lowerarm_r_04_jnt"),Transform=(Rotation=(X=0,Y=0,Z=0,W=1),Translation=(X=0,Y=0,Z=0),Scale3D=(X=1,Y=1,Z=1))))
    ItemToModify.0.Transform.Rotation=&RigVMFunction_MathQuaternionFromEuler(Euler=(X=5,Y=5,Z=5), RotationOrder=XYZ).Result
    Weight=@n15.OutputParam
   )
   Subtract = &RigVMFunction_MathDoubleSub(
    A=&RigVMFunction_MathDoubleMul(A=@n15.OutputParam, B=RigVMVariableNode() {variable=Is not playing emote Alpha}.Value).Result
    B=RigVMVariableNode() {variable=IsPistol}.Value
   )
   n16 = RigVMVariableNode() {variable=IsRifle}
   Clamp = &RigVMFunction_MathDoubleClamp(Value=&RigVMFunction_MathDoubleSub(A=&RigVMFunction_MathDoubleSub(A=@Subtract.Result, B=@n16.Value).Result, B=RigVMVariableNode() {variable=Is Gauntlet}.Value).Result)
   > &RigUnit_OffsetTransformForItem(
    Item=(Type=Bone,Name="lowerarm_r_offset")
    OffsetTransform=(Rotation=(X=0,Y=0,Z=0,W=1),Translation=(X=-5,Y=0,Z=-5),Scale3D=(X=1,Y=1,Z=1))
    bPropagateToChildren=True
    Weight=@Clamp.Result
   )
   &RigUnit_ModifyTransforms(
    ExecutePin=()
    ItemToModify=((Item=(Type=Bone,Name="extra_lowerarm_r_01_jnt"),Transform=(Rotation=(X=0,Y=0,Z=0,W=1),Translation=(X=0,Y=0,Z=0),Scale3D=(X=1,Y=1,Z=1))))
    ItemToModify.0.Transform.Rotation=&RigVMFunction_MathQuaternionFromEuler(Euler=(X=0,Y=5,Z=-5)).Result
    Weight=@Multiply_2_2.Result
   )
   > &RigUnit_ModifyTransforms(
    ItemToModify=((Item=(Type=Bone,Name="extra_lowerarm_r_02_jnt"),Transform=(Rotation=(X=0,Y=0,Z=0,W=1),Translation=(X=0,Y=0,Z=0),Scale3D=(X=1,Y=1,Z=1))))
    ItemToModify.0.Transform.Rotation=&RigVMFunction_MathQuaternionFromEuler(Euler=(X=0,Y=5,Z=0)).Result
    Weight=@Multiply_2_2.Result
   )
   > &RigUnit_ModifyTransforms(
    ItemToModify=((Item=(Type=Bone,Name="extra_lowerarm_r_03_jnt"),Transform=(Rotation=(X=0,Y=0,Z=0,W=1),Translation=(X=0,Y=0,Z=0),Scale3D=(X=1,Y=1,Z=1))))
    ItemToModify.0.Transform.Rotation=&RigVMFunction_MathQuaternionFromEuler(Euler=(X=0,Y=-10,Z=0)).Result
    Weight=@Multiply_2_2.Result
   )
   > &RigUnit_ModifyTransforms(
    ItemToModify=((Item=(Type=Bone,Name="extra_lowerarm_r_04_jnt"),Transform=(Rotation=(X=0,Y=0,Z=0,W=1),Translation=(X=0,Y=0,Z=0),Scale3D=(X=1,Y=1,Z=1))))
    ItemToModify.0.Transform.Rotation=&RigVMFunction_MathQuaternionFromEuler(Euler=(X=0,Y=-10,Z=0)).Result
    Weight=@Multiply_2_2.Result
   )
  ]}
 New Function_ContainedGraph {nodes [
   n1 = RigVMFunctionEntryNode()
   RigVMFunction_Sequence = &RigVMFunction_Sequence(ExecuteContext=@n1.ExecuteContext)
   &RigUnit_ParentConstraint(
    AdvancedSettings=(InterpolationType=Average,RotationOrderForFilter=XZY)
    Child=(Type=Bone,Name="dyn_loincloth_MSTR")
    Filter=(TranslationFilter=(bX=true,bY=false,bZ=false),RotationFilter=(bX=false,bY=True,bZ=True),ScaleFilter=(bX=True,bY=True,bZ=True))
    Parents=((Item=(Type=Bone,Name="thigh_l"),Weight=0.95),(Item=(Type=Bone,Name="pelvis"),Weight=0.05))
    Weight=1
    ExecutePin=@RigVMFunction_Sequence.A
   )
   > &RigUnit_ModifyTransforms(
    ItemToModify=((Item=(Type=Bone,Name="dyn_loincloth_MSTR"),Transform=(Rotation=(X=0.0436,Y=0,Z=0,W=0.999),Translation=(X=0,Y=0,Z=1),Scale3D=(X=1,Y=1,Z=1))))
    Weight=1
    WeightMinimum=0
   )
   > &RigUnit_ModifyTransforms(
    ItemToModify=((Item=(Type=Bone,Name="knife_jnt"),Transform=(Rotation=(X=0.0436,Y=-0.0044,Z=0.0002,W=0.999),Translation=(X=0,Y=-0.5,Z=0),Scale3D=(X=1,Y=1,Z=1))))
    Weight=1
    WeightMinimum=0
   )
   > n2 = &RigUnit_SphericalPoseReader(
    ActiveRegionScaleFactors=(PositiveWidth=4,NegativeWidth=3,PositiveHeight=2.5,NegativeHeight=2)
    ActiveRegionSize=0.1
    Debug=(bDrawDebug=false,bDraw2D=False,bDrawLocalAxes=False,DebugScale=25,DebugSegments=20,DebugThickness=0.25)
    DriverAxis=(X=-1,Y=0,Z=0)
    DriverItem=(Type=Bone,Name="thigh_l")
    FalloffRegionScaleFactors=(PositiveWidth=1,NegativeWidth=4,PositiveHeight=1,NegativeHeight=0.5)
    FalloffSize=0.4
    FlipHeightScaling=False
    FlipWidthScaling=False
    OptionalParentItem=(Type=Bone,Name="None")
    RotationOffset=(X=100,Y=0,Z=-20)
   )
   > &RigUnit_ParentConstraint(
    AdvancedSettings=(InterpolationType=Average,RotationOrderForFilter=XZY)
    Child=(Type=Bone,Name="dyn_loincloth_MSTR")
    Filter=(TranslationFilter=(bX=true,bY=true,bZ=true),RotationFilter=(bX=true,bY=True,bZ=True),ScaleFilter=(bX=True,bY=True,bZ=True))
    Parents=((Item=(Type=Bone,Name="thigh_l"),Weight=0.95),(Item=(Type=Bone,Name="pelvis"),Weight=0.05))
    Weight=@n2.OutputParam
   )
   > &RigUnit_ModifyTransforms(
    ItemToModify=((Item=(Type=Bone,Name="knife_jnt"),Transform=(Rotation=(X=0.1798,Y=0.0704,Z=-0.1006,W=0.976),Translation=(X=0,Y=-5,Z=0),Scale3D=(X=1,Y=1,Z=1))))
    WeightMinimum=0
    Weight=@n2.OutputParam
   )
   > &RigUnit_ModifyTransforms(
    ItemToModify=((Item=(Type=Bone,Name="dyn_loincloth_MSTR"),Transform=(Rotation=(X=-0.1736,Y=0.0215,Z=0.0038,W=0.9846),Translation=(X=2,Y=0,Z=0.5),Scale3D=(X=1,Y=1,Z=1))))
    WeightMinimum=0
    Weight=@n2.OutputParam
   )
   RigVMFunctionReturnNode(ExecuteContext=@RigVMFunction_Sequence.B)
  ]}
 RigVMFunctionLibrary {nodes [
   Apron Logic = RigVMCollapseNode(ExecuteContext=())
   Lowerand front TentacleArms = RigVMCollapseNode(ExecuteContext=())
   Proto Deform Tentacle Finger = RigVMCollapseNode(ExecuteContext=())
   TentacleArmSpline = RigVMCollapseNode(ExecuteContext=())
  ]}
 RigVMFunction_Sequence_1_ContainedGraph {nodes [
   n1 = RigVMFunctionEntryNode()
   RigVMFunction_Sequence = &RigVMFunction_Sequence(ExecuteContext=@n1.ExecuteContext)
   n2 = &RigVMFunction_Sequence(ExecuteContext=@RigVMFunction_Sequence.B)
   n3 = &RigVMFunction_Sequence(ExecuteContext=@n2.B)
   n4 = &RigVMFunction_Sequence(ExecuteContext=@n3.B)
   RigVMFunctionReturnNode(A=@RigVMFunction_Sequence.A, B=@n2.A, C=@n3.A, D=@n4.A, E=@n4.B)
  ]}
 RigVMFunction_Sequence_1_ContainedGraph {nodes [
   n1 = RigVMFunctionEntryNode()
   RigVMFunction_Sequence = &RigVMFunction_Sequence(ExecuteContext=@n1.ExecuteContext)
   n2 = &RigVMFunction_Sequence(ExecuteContext=@RigVMFunction_Sequence.B)
   RigVMFunctionReturnNode(A=@RigVMFunction_Sequence.A, B=@n2.A, C=@n2.B)
  ]}
 RigVMFunction_Sequence_1_ContainedGraph {nodes [
   n1 = RigVMFunctionEntryNode()
   RigVMFunction_Sequence = &RigVMFunction_Sequence(ExecuteContext=@n1.ExecuteContext)
   n2 = &RigVMFunction_Sequence(ExecuteContext=@RigVMFunction_Sequence.B)
   n3 = &RigVMFunction_Sequence(ExecuteContext=@n2.B)
   n4 = &RigVMFunction_Sequence(ExecuteContext=@n3.B)
   n5 = &RigVMFunction_Sequence(ExecuteContext=@n4.B)
   RigVMFunctionReturnNode(A=@RigVMFunction_Sequence.A, B=@n2.A, C=@n3.A, D=@n4.A, E=@n5.A, F=@n5.B)
  ]}
 RigVMModel {nodes [
   RigUnit_BeginExecution = &RigUnit_BeginExecution()
   FN_DeformRig_FNC = RigVMFunctionReferenceNode(
    DeformRig_ Curve="(deformRig_l_upperarm_up_alpha_72_#1=False,deformRig_r_upperarm_up_alpha_73_#2=False,deformRig_l_upperarm_fr_alpha_74_#3=False,deformRig_r_upperarm_fr_alpha_75_CB51B8 ...(truncated)"
    DeformRig_Bool="(upperarm_1_#4=True,elbow_3_#5=True,wrist_7_#6=True,groin_8_#7=True,glute_17_#8=True,knee_9_40C408324CA261 ...(truncated)"
    DeformRig_Debug=(upperarm_debug_18_#4=False,wrist_debug_20_#6=False,lat_debug_26_#9=False,groinGlute_debug_27_#7=False)
    DeformRig_Value="(upperarm_up_18_#10=(Rotation=(X=0.0,Y=-0.0,Z=0.0,W=1.0),Translation=(X=0.0,Y=0.0,Z=8.0),Scale3D=(X=0.70,Y=1.0,Z=1.0)),upperarm_fwd_fr_19_#11=(Rotation=(X=0.0,Y=-0.0,Z=0.0,W=1.0),Translation=(X=0.0, ...(truncated)"
    ExecuteContext=@RigUnit_BeginExecution.ExecutePin
   )
   > Lowerand front TentacleArms = RigVMFunctionReferenceNode()
   > TentacleArmSpline = RigVMFunctionReferenceNode()
   > Proto Deform Tentacle Finger = RigVMFunctionReferenceNode()
   > Apron Logic = RigVMFunctionReferenceNode()
   > n1 = &RigUnit_SphericalPoseReader(
    ActiveRegionScaleFactors=(PositiveWidth=1,NegativeWidth=1,PositiveHeight=1,NegativeHeight=1)
    ActiveRegionSize=0.1
    Debug=(bDrawDebug=false,bDraw2D=False,bDrawLocalAxes=False,DebugScale=25,DebugSegments=20,DebugThickness=0.25)
    DriverAxis=(X=-1,Y=0,Z=0)
    DriverItem=(Type=Bone,Name="calf_l")
    FalloffRegionScaleFactors=(PositiveWidth=1,NegativeWidth=1,PositiveHeight=1,NegativeHeight=1)
    FalloffSize=0.2
    FlipHeightScaling=false
    FlipWidthScaling=false
    OptionalParentItem=(Type=Bone,Name="None")
    RotationOffset=(X=-90,Y=0,Z=65)
   )
   > &RigUnit_ModifyTransforms(
    ItemToModify=((Item=(Type=Bone,Name="calf_volume_l"),Transform=(Rotation=(X=0,Y=0,Z=-0.1736,W=0.9848),Translation=(X=0,Y=0,Z=0),Scale3D=(X=1,Y=1,Z=1))))
    WeightMinimum=0
    Weight=@n1.OutputParam
   )
   > n2 = &RigUnit_SphericalPoseReader(
    ActiveRegionScaleFactors=(PositiveWidth=1,NegativeWidth=1,PositiveHeight=1,NegativeHeight=1)
    ActiveRegionSize=0.1
    Debug=(bDrawDebug=false,bDraw2D=False,bDrawLocalAxes=False,DebugScale=25,DebugSegments=20,DebugThickness=0.25)
    DriverAxis=(X=1,Y=0,Z=0)
    DriverItem=(Type=Bone,Name="calf_r")
    FalloffRegionScaleFactors=(PositiveWidth=1,NegativeWidth=1,PositiveHeight=1,NegativeHeight=1)
    FalloffSize=0.2
    FlipHeightScaling=False
    FlipWidthScaling=true
    OptionalParentItem=(Type=Bone,Name="None")
    RotationOffset=(X=90,Y=0,Z=65)
   )
   > &RigUnit_ModifyTransforms(
    ItemToModify=((Item=(Type=Bone,Name="calf_volume_r"),Transform=(Rotation=(X=0,Y=0,Z=-0.1736,W=0.9848),Translation=(X=0,Y=0,Z=0),Scale3D=(X=1,Y=1,Z=1))))
    WeightMinimum=0
    Weight=@n2.OutputParam
   )
  ]}
]
hierarchy [
 root {loc=<0 0 0>, rot=<0 0 0 1>}
 pelvis {parent=root, loc=<0 -0.7071 93.1191>, rot=<0.0127 0.707 -0.0127 -0.707>}
 spine_01 {parent=pelvis, loc=<6.7494 -0.6388 0>, rot=<0 0 0.034 -0.9994>}
 spine_02 {parent=spine_01, loc=<7.2518 -0.035 0>, rot=<0 0 0.0238 -0.9997>}
 spine_03 {parent=spine_02, loc=<6.0104 0 0>, rot=<0 0 0.0149 -0.9999>}
 spine_04 {parent=spine_03, loc=<8.1024 0 0>, rot=<0 0 -0.0132 -0.9999>}
 spine_05 {parent=spine_04, loc=<10.0146 0 0>, rot=<0 0 -0.0015 -1>}
 neck_01 {parent=spine_05, loc=<11.3138 0.1529 0>, rot=<0 0 -0.2003 -0.9797>}
 neck_02 {parent=neck_01, loc=<4.5627 0 0>, rot=<0 0 0.0504 -0.9987>}
 head {parent=neck_02, loc=<5.3914 0 0>, rot=<0 0 0.115 -0.9934>}
 attach_cape {parent=spine_05, loc=<9.017 -11.714 0>, rot=<0.0616 -0.7044 -0.0616 -0.7044>}
 cape {parent=attach_cape, type=Null, loc=<0 0 0>, rot=<0 0 0 1>}
 attach_backpack {parent=spine_05, loc=<-0.123 -13.473 0>, rot=<-0.0093 -0.707 0.0093 -0.707>}
 Backpack_BR {parent=attach_backpack, type=Null, loc=<0 -1.309 0>, rot=<-0.0436 0 0 0.999>}
 clavicle_l {parent=spine_05, loc=<4.3575 -0.1466 -2.3223>, rot=<0.0465 -0.7289 -0.009 -0.6829>}
 upperarm_l {parent=clavicle_l, loc=<15.5111 0 0>, rot=<0.0226 -0.3234 -0.0088 -0.9459>}
 lowerarm_l {parent=upperarm_l, loc=<26.6514 0 0>, rot=<0.0391 -0.0099 -0.2453 -0.9686>}
 hand_l {parent=lowerarm_l, loc=<22.3491 0 0>, rot=<-0.7081 -0.0507 0.0975 -0.6975>}
 thumb_01_l {parent=hand_l, loc=<1.3118 -0.3191 -2.459>, rot=<0.6546 -0.4042 0.0664 -0.6354>}
 thumb_02_l {parent=thumb_01_l, loc=<5.9336 -0.0448 0.0401>, rot=<-0.0037 0.0213 0.298 -0.9543>}
 thumb_03_l {parent=thumb_02_l, loc=<2.2223 0.3682 -0.0907>, rot=<0.0118 -0.013 0.0066 -0.9998>}
 index_metacarpal_l {parent=hand_l, loc=<1.6535 0.0183 -2.96>, rot=<0.0158 -0.0124 -0.0173 -0.9996>}
 index_01_l {parent=index_metacarpal_l, loc=<8.3008 0.0357 -0.0524>, rot=<-0.0044 0.0485 0.0924 -0.9945>}
 index_02_l {parent=index_01_l, loc=<3.3587 -0.3004 0.0452>, rot=<-0.002 0.0027 0.3225 -0.9466>}
 index_03_l {parent=index_02_l, loc=<2.5578 -0.123 -0.0252>, rot=<-0.002 0.0067 0.2558 -0.9667>}
 middle_metacarpal_l {parent=hand_l, loc=<2.1856 -0.1725 -0.5984>, rot=<0.006 0.0168 -0.0329 -0.9993>}
 middle_01_l {parent=middle_metacarpal_l, loc=<7.3275 -0.2447 0.0129>, rot=<-0.0208 0.0042 0.0858 -0.9961>}
 middle_02_l {parent=middle_01_l, loc=<4.0123 0.1401 0.1175>, rot=<-0.0218 0.0258 0.3512 -0.9357>}
 middle_03_l {parent=middle_02_l, loc=<2.6077 0.0534 0.1373>, rot=<0.0082 0.0188 0.1708 -0.9851>}
 ring_metacarpal_l {parent=hand_l, loc=<2.2074 -0.3625 1.3638>, rot=<-0.0284 0.0722 -0.0458 -0.9959>}
 ring_01_l {parent=ring_metacarpal_l, loc=<7.0295 -0.6197 -0.1446>, rot=<-0.0489 -0.0252 0.1892 -0.9804>}
 ring_02_l {parent=ring_01_l, loc=<3.6174 0.4647 -0.1027>, rot=<0 0 0.2806 -0.9598>}
 ring_03_l {parent=ring_02_l, loc=<2.4152 0.114 0.0199>, rot=<0 0 0.1189 -0.9929>}
 pinky_metacarpal_l {parent=hand_l, loc=<1.7616 -0.9753 3.3814>, rot=<-0.0397 0.0814 -0.069 -0.9935>}
 pinky_01_l {parent=pinky_metacarpal_l, loc=<7.1709 -0.4831 -0.1094>, rot=<-0.0654 -0.0035 0.2242 -0.9723>}
 pinky_02_l {parent=pinky_01_l, loc=<2.8274 0.0893 -0.1043>, rot=<0.0037 0.0011 0.2854 -0.9584>}
 pinky_03_l {parent=pinky_02_l, loc=<2.2046 0.0307 -0.0078>, rot=<-0.0002 0.0098 0.1445 -0.9895>}
 deform_tentacle_finger_01_l {parent=hand_l, loc=<2.4961 0 -0.9798>, rot=<-0.001 -0.068 -0.0141 -0.9976>}
 deform_tentacle_finger_02_l {parent=deform_tentacle_finger_01_l, loc=<3.5 0 0>, rot=<0 0 0 1>}
 deform_tentacle_finger_03_l {parent=deform_tentacle_finger_02_l, loc=<3.5 0 0>, rot=<0 0 0.1024 -0.9947>}
 deform_tentacle_finger_04_l {parent=deform_tentacle_finger_03_l, loc=<3.5 0 0>, rot=<0 0 0.2517 -0.9678>}
 deform_tentacle_finger_05_l {parent=deform_tentacle_finger_04_l, loc=<3.5 0 0>, rot=<0 0 0.0972 -0.9953>}
 deform_tentacle_finger_06_l {parent=deform_tentacle_finger_05_l, loc=<2 0 0>, rot=<0 0 0 1>}
 cr_index_metacarpal_l {parent=hand_l, loc=<2.326 1.5788 0.6383>, rot=<-0.1112 0.0816 -0.1368 -0.981>}
 cr_index_01_l {parent=cr_index_metacarpal_l, loc=<8.3008 0.0357 -0.0524>, rot=<0 0 0 1>}
 cr_index_02_l {parent=cr_index_01_l, loc=<3.3587 -0.3004 0.0452>, rot=<0 0 0 1>}
 cr_index_03_l {parent=cr_index_02_l, loc=<2.5578 -0.123 -0.0252>, rot=<0 0 0 1>}
 lowerarm_twist_02_l {parent=lowerarm_l, loc=<9.6199 0 0>, rot=<0 0 0 -1>}
 lowerarm_twist_01_l {parent=lowerarm_l, loc=<17.8549 0 0>, rot=<0 0 0 -1>}
 lowerarm_l_offset {parent=lowerarm_l, loc=<0 0 0>, rot=<0 0 0 1>}
 upperarm_twist_01_l {parent=upperarm_l, loc=<0.0041 0 0>, rot=<0 0 0 -1>}
 deform_upperarm_bk_base_l {parent=upperarm_twist_01_l, loc=<1.0813 -7.2193 -1.9453>, rot=<0 0 0 1>}
 deform_upperarm_bk_l {parent=deform_upperarm_bk_base_l, loc=<0 0 0>, rot=<0 0 0 1>}
 deform_upperarm_fr_base_l {parent=upperarm_twist_01_l, loc=<-0.3691 7.0506 -2.765>, rot=<0 0 0 1>}
 deform_upperarm_fr_l {parent=deform_upperarm_fr_base_l, loc=<0 0 0>, rot=<0 0 0 1>}
 deform_upperarm_up_base_l {parent=upperarm_twist_01_l, loc=<-3.2216 -1.1391 3.3311>, rot=<0 0 0 1>}
 deform_upperarm_up_l {parent=deform_upperarm_up_base_l, loc=<0 0 0>, rot=<0 0 0 1>}
 upperarm_twist_02_l {parent=upperarm_l, loc=<13.7941 0 0>, rot=<0 0 0 -1>}
 tentacle_01_l {parent=clavicle_l, loc=<15.5143 0 -0.0025>, rot=<0.0226 -0.3234 -0.0088 -0.9459>}
 tentacle_02_l {parent=tentacle_01_l, loc=<6.6569 0 0>, rot=<0 0 -0.0218 -0.9998>}
 tentacle_03_l {parent=tentacle_02_l, loc=<6.6569 0 0>, rot=<0 0 -0.0218 -0.9998>}
 tentacle_04_l {parent=tentacle_03_l, loc=<6.6569 0 0>, rot=<0 0 -0.0218 -0.9998>}
 tentacle_05_l {parent=tentacle_04_l, loc=<6.6569 0 0>, rot=<0.0002 -0.0094 -0.0218 -0.9997>}
 tentacle_06_l {parent=tentacle_05_l, loc=<4.338 0 0>, rot=<0.0013 -0.0093 -0.1345 -0.9909>}
 tentacle_07_l {parent=tentacle_06_l, loc=<4.338 0 0>, rot=<0.0002 -0.0094 -0.0218 -0.9997>}
 tentacle_08_l {parent=tentacle_07_l, loc=<4.338 0 0>, rot=<0.0002 -0.0094 -0.0218 -0.9997>}
 tentacle_09_l {parent=tentacle_08_l, loc=<4.338 0 0>, rot=<0.0002 -0.0094 -0.0218 -0.9997>}
 tentacle_10_l {parent=tentacle_09_l, loc=<4.338 0 0>, rot=<0.0002 -0.0094 -0.0218 -0.9997>}
 clavicle_r {parent=spine_05, loc=<4.3575 -0.1466 2.3223>, rot=<0.7289 0.0465 -0.6829 0.009>}
 upperarm_r {parent=clavicle_r, loc=<-15.5111 0 0>, rot=<0.0226 -0.3234 -0.0088 -0.9459>}
 lowerarm_r {parent=upperarm_r, loc=<-26.6514 0 0>, rot=<0.0391 -0.0099 -0.2453 -0.9686>}
 hand_r {parent=lowerarm_r, loc=<-22.3491 0 0>, rot=<-0.7081 -0.0507 0.0975 -0.6975>}
 thumb_01_r {parent=hand_r, loc=<-1.3118 0.3191 2.459>, rot=<0.6546 -0.4042 0.0664 -0.6354>}
 thumb_02_r {parent=thumb_01_r, loc=<-5.9336 0.0448 -0.0401>, rot=<-0.0037 0.0213 0.298 -0.9543>}
 thumb_03_r {parent=thumb_02_r, loc=<-2.2223 -0.3682 0.0907>, rot=<0.0118 -0.013 0.0066 -0.9998>}
 index_metacarpal_r {parent=hand_r, loc=<-1.6535 -0.0183 2.96>, rot=<0.0158 -0.0124 -0.0173 -0.9996>}
 index_01_r {parent=index_metacarpal_r, loc=<-8.3008 -0.0357 0.0524>, rot=<-0.0044 0.0485 0.0924 -0.9945>}
 index_02_r {parent=index_01_r, loc=<-3.3587 0.3004 -0.0452>, rot=<-0.002 0.0027 0.3225 -0.9466>}
 index_03_r {parent=index_02_r, loc=<-2.5578 0.123 0.0252>, rot=<-0.002 0.0067 0.2558 -0.9667>}
 middle_metacarpal_r {parent=hand_r, loc=<-2.1856 0.1725 0.5984>, rot=<0.006 0.0168 -0.0329 -0.9993>}
 middle_01_r {parent=middle_metacarpal_r, loc=<-7.3275 0.2447 -0.0129>, rot=<-0.0208 0.0042 0.0858 -0.9961>}
 middle_02_r {parent=middle_01_r, loc=<-4.0123 -0.1401 -0.1175>, rot=<-0.0218 0.0258 0.3512 -0.9357>}
 middle_03_r {parent=middle_02_r, loc=<-2.6077 -0.0534 -0.1373>, rot=<0.0082 0.0188 0.1708 -0.9851>}
 ring_metacarpal_r {parent=hand_r, loc=<-2.2074 0.3625 -1.3638>, rot=<-0.0284 0.0722 -0.0458 -0.9959>}
 ring_01_r {parent=ring_metacarpal_r, loc=<-7.0295 0.6197 0.1446>, rot=<-0.0489 -0.0252 0.1892 -0.9804>}
 ring_02_r {parent=ring_01_r, loc=<-3.6174 -0.4647 0.1027>, rot=<0 0 0.2806 -0.9598>}
 ring_03_r {parent=ring_02_r, loc=<-2.4152 -0.114 -0.0199>, rot=<0 0 0.1189 -0.9929>}
 pinky_metacarpal_r {parent=hand_r, loc=<-1.7616 0.9753 -3.3814>, rot=<-0.0397 0.0814 -0.069 -0.9935>}
 pinky_01_r {parent=pinky_metacarpal_r, loc=<-7.1709 0.4831 0.1094>, rot=<-0.0654 -0.0035 0.2242 -0.9723>}
 pinky_02_r {parent=pinky_01_r, loc=<-2.8274 -0.0893 0.1043>, rot=<0.0037 0.0011 0.2854 -0.9584>}
 pinky_03_r {parent=pinky_02_r, loc=<-2.2046 -0.0307 0.0078>, rot=<-0.0002 0.0098 0.1445 -0.9895>}
 deform_tentacle_finger_01_r {parent=hand_r, loc=<-2.4959 -0.0002 0.9799>, rot=<-0.001 -0.068 -0.0141 -0.9976>}
 deform_tentacle_finger_02_r {parent=deform_tentacle_finger_01_r, loc=<-3.5002 0.0003 -0.0001>, rot=<0 0 0 1>}
 deform_tentacle_finger_03_r {parent=deform_tentacle_finger_02_r, loc=<-3.5 -0.0001 0>, rot=<0 0 0.1024 -0.9947>}
 deform_tentacle_finger_04_r {parent=deform_tentacle_finger_03_r, loc=<-3.5 0 0>, rot=<0 0 0.2517 -0.9678>}
 deform_tentacle_finger_05_r {parent=deform_tentacle_finger_04_r, loc=<-3.5 0 0>, rot=<0 0 0.0972 -0.9953>}
 deform_tentacle_finger_06_r {parent=deform_tentacle_finger_05_r, loc=<-2 -0.0001 0>, rot=<0 0 0 -1>}
 cr_index_metacarpal_r {parent=hand_r, loc=<-2.3263 -1.5785 -0.6384>, rot=<-0.1112 0.0816 -0.1368 -0.981>}
 cr_index_01_r {parent=cr_index_metacarpal_r, loc=<-8.3008 -0.0357 0.0524>, rot=<0 0 0 1>}
 cr_index_02_r {parent=cr_index_01_r, loc=<-3.3587 0.3004 -0.0452>, rot=<0 0 0 1>}
 cr_index_03_r {parent=cr_index_02_r, loc=<-2.5578 0.123 0.0252>, rot=<0 0 0 1>}
 lowerarm_twist_02_r {parent=lowerarm_r, loc=<-9.6199 0 0>, rot=<0 0 0 -1>}
 lowerarm_twist_01_r {parent=lowerarm_r, loc=<-17.8549 0 0>, rot=<0 0 0 -1>}
 lowerarm_r_offset {parent=lowerarm_r, loc=<0 0 0>, rot=<0 0 0 1>}
 upperarm_twist_01_r {parent=upperarm_r, loc=<-0.0041 0 0>, rot=<0 0 0 -1>}
 deform_upperarm_bk_base_r {parent=upperarm_twist_01_r, loc=<-1.0814 7.2192 1.9455>, rot=<0 0 0 1>}
 deform_upperarm_bk_r {parent=deform_upperarm_bk_base_r, loc=<0 0 0>, rot=<0 0 0 1>}
 deform_upperarm_fr_base_r {parent=upperarm_twist_01_r, loc=<0.3689 -7.0506 2.7652>, rot=<0 0 0 1>}
 deform_upperarm_fr_r {parent=deform_upperarm_fr_base_r, loc=<0 0 0>, rot=<0 0 0 1>}
 deform_upperarm_up_base_r {parent=upperarm_twist_01_r, loc=<3.2214 1.1391 -3.3308>, rot=<0 0 0 1>}
 deform_upperarm_up_r {parent=deform_upperarm_up_base_r, loc=<0 0 0>, rot=<0 0 0 -1>}
 upperarm_twist_02_r {parent=upperarm_r, loc=<-13.7941 0 0>, rot=<0 0 0 -1>}
 tentacle_01_r {parent=clavicle_r, loc=<-15.5144 0 0.0029>, rot=<0.0226 -0.3234 -0.0088 -0.9459>}
 tentacle_02_r {parent=tentacle_01_r, loc=<-6.6568 0 0>, rot=<0 0 -0.0218 -0.9998>}
 tentacle_03_r {parent=tentacle_02_r, loc=<-6.657 0 0>, rot=<0 0 -0.0218 -0.9998>}
 tentacle_04_r {parent=tentacle_03_r, loc=<-6.6566 0 -0.0003>, rot=<0 0 -0.0218 -0.9998>}
 tentacle_05_r {parent=tentacle_04_r, loc=<-6.6571 0 0.0003>, rot=<0.0002 -0.0094 -0.0218 -0.9997>}
 tentacle_06_r {parent=tentacle_05_r, loc=<-4.3377 0 -0.0004>, rot=<0.0013 -0.0093 -0.1345 -0.9909>}
 tentacle_07_r {parent=tentacle_06_r, loc=<-4.338 0 0>, rot=<0.0002 -0.0094 -0.0218 -0.9997>}
 tentacle_08_r {parent=tentacle_07_r, loc=<-4.3378 -0.0001 -0.0002>, rot=<0.0002 -0.0094 -0.0218 -0.9997>}
 tentacle_09_r {parent=tentacle_08_r, loc=<-4.3384 0.0002 0.0004>, rot=<0.0002 -0.0094 -0.0218 -0.9997>}
 tentacle_10_r {parent=tentacle_09_r, loc=<-4.3378 -0.0001 -0.0002>, rot=<0.0002 -0.0094 -0.0218 -0.9997>}
 chest_arms_ft_jnt {parent=spine_04, loc=<3.7889 15.1126 0>, rot=<0.4788 0.5204 0.5204 0.4788>}
 chest_arms_ft_l_MSTR {parent=chest_arms_ft_jnt, loc=<3.9773 10.4548 -3.0524>, rot=<0.1783 0.0777 0.7141 -0.6725>}
 chest_arms_ft_l_01_jnt {parent=chest_arms_ft_l_MSTR, loc=<0 0 0>, rot=<0 0 0 1>}
 chest_arms_ft_l_02_jnt {parent=chest_arms_ft_l_01_jnt, loc=<5.0431 0 0>, rot=<0.0038 0.0406 0.0942 -0.9947>}
 chest_arms_ft_l_03_jnt {parent=chest_arms_ft_l_02_jnt, loc=<5.0431 0 0>, rot=<-0.0006 -0.0063 0.0878 -0.9961>}
 chest_arms_ft_l_04_jnt {parent=chest_arms_ft_l_03_jnt, loc=<5.0431 0 0>, rot=<0 -0.1209 0 -0.9927>}
 chest_arms_ft_l_05_jnt {parent=chest_arms_ft_l_04_jnt, loc=<5.0431 0 0>, rot=<0 -0.1409 0 -0.99>}
 chest_arms_ft_l_Offset_MSTR {parent=chest_arms_ft_l_MSTR, loc=<0 0 0>, rot=<0 0 0 1>}
 chest_arms_ft_r_MSTR {parent=chest_arms_ft_jnt, loc=<-0.0579 5.9403 1.9541>, rot=<-0.1106 0.0832 -0.7484 -0.6486>}
 chest_arms_ft_r_01_jnt {parent=chest_arms_ft_r_MSTR, loc=<0 0 0>, rot=<0 0 0 1>}
 chest_arms_ft_r_02_jnt {parent=chest_arms_ft_r_01_jnt, loc=<3.7649 0 0>, rot=<-0.0149 -0.0308 -0.0671 -0.9972>}
 chest_arms_ft_r_03_jnt {parent=chest_arms_ft_r_02_jnt, loc=<3.7649 0 0>, rot=<0.0702 0.6355 -0.1051 -0.7617>}
 chest_arms_ft_r_04_jnt {parent=chest_arms_ft_r_03_jnt, loc=<2.263 0 0>, rot=<-0.0276 -0.1392 0.1926 -0.971>}
 chest_arms_ft_r_05_jnt {parent=chest_arms_ft_r_04_jnt, loc=<8.5698 0 0>, rot=<0 0 0 1>}
 extra_lowerarm_l_01_jnt {parent=spine_03, loc=<-2.2915 -3.8556 -20.0557>, rot=<-0.3094 -0.6656 -0.4463 -0.5119>}
 extra_lowerarm_l_02_jnt {parent=extra_lowerarm_l_01_jnt, loc=<4.5662 0 0>, rot=<-0.0027 0.0157 -0.1672 -0.9858>}
 extra_lowerarm_l_03_jnt {parent=extra_lowerarm_l_02_jnt, loc=<4.5662 0 0>, rot=<-0.0027 0.0157 -0.1672 -0.9858>}
 extra_lowerarm_l_04_jnt {parent=extra_lowerarm_l_03_jnt, loc=<4.5662 0 0>, rot=<-0.0027 0.0157 -0.1672 -0.9858>}
 extra_lowerarm_l_05_jnt {parent=extra_lowerarm_l_04_jnt, loc=<4.5662 0 0>, rot=<-0.0289 0.1705 -0.1647 -0.9711>}
 extra_lowerarm_l_06_jnt {parent=extra_lowerarm_l_05_jnt, loc=<6.2225 0 0>, rot=<0 -0.1913 0 -0.9815>}
 dyn_lowerarm_Sleeve_l_MSTR {parent=extra_lowerarm_l_02_jnt, loc=<4.5662 0 0>, rot=<0.1517 -0.676 -0.0829 -0.7163>}
 dyn_lowerarm_Sleeve_l_01_jnt {parent=dyn_lowerarm_Sleeve_l_MSTR, loc=<0 0 0>, rot=<0 0 0 1>}
 dyn_lowerarm_Sleeve_l_02_jnt {parent=dyn_lowerarm_Sleeve_l_01_jnt, loc=<10.4964 0 0>, rot=<0 0 0 1>}
 extra_lowerarm_r_01_jnt {parent=spine_03, loc=<-2.2915 -3.8556 20.0557>, rot=<0.4975 0.4623 0.6555 0.3303>}
 extra_lowerarm_r_02_jnt {parent=extra_lowerarm_r_01_jnt, loc=<6.5058 0 0.0003>, rot=<0 0 0.1549 -0.9879>}
 extra_lowerarm_r_03_jnt {parent=extra_lowerarm_r_02_jnt, loc=<6.5058 0 -0.0001>, rot=<0 0 0.1549 -0.9879>}
 extra_lowerarm_r_04_jnt {parent=extra_lowerarm_r_03_jnt, loc=<6.5058 -0.0001 0.0003>, rot=<0 0 0.1549 -0.9879>}
 extra_lowerarm_r_05_jnt {parent=extra_lowerarm_r_04_jnt, loc=<6.5058 0.0001 -0.0004>, rot=<-0.005 -0.01 0.4431 -0.8964>}
 extra_lowerarm_r_06_jnt {parent=extra_lowerarm_r_05_jnt, loc=<6.5058 0 -2.3287>, rot=<-0.1061 -0.6765 0.1129 -0.72>}
 dyn_gourd_handle_jnt {parent=extra_lowerarm_r_06_jnt, loc=<1.4878 0.2573 -4.5126>, rot=<-0.1397 -0.114 -0.1655 -0.9696>}
 dyn_gourd_Base_jnt {parent=dyn_gourd_handle_jnt, loc=<4.6601 0 0>, rot=<-0.0196 0.2288 0.0829 -0.9697>}
 dyn_gourd_jnt {parent=dyn_gourd_Base_jnt, loc=<3.798 0 0>, rot=<0 0.1673 0 -0.9859>}
 dyn_lowerarm_Sleeve_r_MSTR {parent=extra_lowerarm_r_04_jnt, loc=<0 0 -0.0003>, rot=<0.0626 -0.69 -0.1403 -0.7073>}
 dyn_lowerarm_Sleeve_r_01_jnt {parent=dyn_lowerarm_Sleeve_r_MSTR, loc=<0 0 0>, rot=<0 0 0 1>}
 dyn_lowerarm_Sleeve_r_02_jnt {parent=dyn_lowerarm_Sleeve_r_01_jnt, loc=<10.4966 0 0>, rot=<0 0 0 1>}
 extra_lowerarm_04pos_l_jnt {parent=spine_03, loc=<-5.615 8.853 -21.1747>, rot=<0.6161 0.413 0.6182 0.2602>}
 extra_lowerarm_03pos_l_jnt {parent=spine_03, loc=<-4.7441 4.5125 -22.2935>, rot=<-0.5278 -0.5126 -0.5768 -0.355>}
 extra_lowerarm_03pos_r_jnt {parent=spine_03, loc=<-5.9862 8.191 22.5989>, rot=<0.3321 0.5924 0.5229 0.5151>}
 extra_lowerarm_04pos_r_jnt {parent=spine_03, loc=<-7.6041 14.2557 20.888>, rot=<-0.2363 -0.6367 -0.4368 -0.5899>}
 extra_lowerarm_05pos_r_jnt {parent=spine_03, loc=<-8.8558 19.566 17.344>, rot=<-0.0717 0.6835 0.1293 0.7148>}
 thigh_l {parent=pelvis, loc=<-2.5693 -0.0109 -7.7501>, rot=<0.0012 0.0572 0.0155 -0.9982>}
 calf_l {parent=thigh_l, loc=<-39.0484 0 0>, rot=<0.0006 -0.0161 -0.0391 -0.9991>}
 foot_l {parent=calf_l, loc=<-42.561 0.1077 0.4292>, rot=<0.0418 -0.0455 0.0343 -0.9975>}
 ball_l {parent=foot_l, loc=<-6.1599 13.8264 0.6806>, rot=<0.0001 -0.0001 -0.7063 -0.7079>}
 calf_twist_02_l {parent=calf_l, loc=<-18.8064 -0.0967 0.0397>, rot=<0.0237 0.0095 -0.0035 -0.9997>}
 calf_twist_01_l {parent=calf_l, loc=<-35.813 -0.0967 0.0397>, rot=<0.0237 0.0095 -0.0035 -0.9997>}
 calf_volume_l {parent=calf_l, loc=<0 0 0>, rot=<0 0 0 1>}
 thigh_twist_01_l {parent=thigh_l, loc=<-21.6303 0 0>, rot=<0.0013 0 0 -1>}
 deform_groin_base_l {parent=thigh_l, loc=<-2.1908 8.1142 -1.7442>, rot=<0 0 0 -1>}
 deform_groin_l {parent=deform_groin_base_l, loc=<0 0 0>, rot=<0 0 0 1>}
 thigh_r {parent=pelvis, loc=<-2.5693 -0.0109 7.7501>, rot=<0.0572 -0.0012 0.9982 0.0155>}
 calf_r {parent=thigh_r, loc=<39.0484 0 0>, rot=<0.0006 -0.0161 -0.0391 -0.9991>}
 foot_r {parent=calf_r, loc=<42.561 -0.1077 -0.4292>, rot=<0.0418 -0.0455 0.0343 -0.9975>}
 ball_r {parent=foot_r, loc=<6.1599 -13.8264 -0.6806>, rot=<0.0001 -0.0001 -0.7063 -0.7079>}
 calf_twist_02_r {parent=calf_r, loc=<18.8064 0.0967 -0.0397>, rot=<0.0237 0.0095 -0.0035 -0.9997>}
 calf_twist_01_r {parent=calf_r, loc=<35.813 0.0967 -0.0397>, rot=<0.0237 0.0095 -0.0035 -0.9997>}
 calf_volume_r {parent=calf_r, loc=<0 0 0>, rot=<0 0 0 -1>}
 thigh_twist_01_r {parent=thigh_r, loc=<21.6303 0 0>, rot=<0.0013 0 0 -1>}
 deform_groin_base_r {parent=thigh_r, loc=<2.1908 -8.1142 1.7442>, rot=<0 0 0 -1>}
 deform_groin_r {parent=deform_groin_base_r, loc=<0 0 0>, rot=<0 0 0 1>}
 deform_glute_base_l {parent=pelvis, loc=<-8.8064 -8.8222 -8.4287>, rot=<0.0275 0.0501 -0.4472 -0.8926>}
 deform_glute_l {parent=deform_glute_base_l, loc=<0 0 0>, rot=<0 0 0 1>}
 deform_glute_base_r {parent=pelvis, loc=<-8.8062 -8.8222 8.4287>, rot=<0.0501 -0.0275 0.8926 -0.4472>}
 deform_glute_r {parent=deform_glute_base_r, loc=<0 0 0>, rot=<0 0 0 1>}
 dyn_loincloth_MSTR {parent=pelvis, loc=<-0.3326 13.2698 -9.5084>, rot=<0.0042 -0.841 -0.5407 -0.0187>}
 dyn_loincloth_01_jnt {parent=dyn_loincloth_MSTR, loc=<0 0 0>, rot=<0 0 0 1>}
 dyn_loincloth_02_jnt {parent=dyn_loincloth_01_jnt, loc=<8.5867 0 0>, rot=<0 0 0 1>}
 dyn_loincloth_03_jnt {parent=dyn_loincloth_02_jnt, loc=<8.5867 0 0>, rot=<0 0 0 1>}
 dyn_loincloth_04_jnt {parent=dyn_loincloth_03_jnt, loc=<8.5867 0 0>, rot=<0 0 0 1>}
 knife_jnt {parent=pelvis, loc=<4.8887 8.514 -16.1211>, rot=<0.31 -0.9187 -0.2333 -0.0737>}
 Handle_aim_jnt {parent=knife_jnt, loc=<-14.6743 -2.6264 1.4178>, rot=<0 0 0 1>}
 extra_lowerarm_06pos_l_jnt {parent=pelvis, loc=<18.1559 15.206 -14.2664>, rot=<-0.6846 -0.258 -0.6531 -0.1954>}
 pelvis_offset {parent=pelvis, loc=<0 0 0>, rot=<0 0 0 1>}
]
variables [
 Is Gauntlet {type=double, default=0}
 Is not playing emote Alpha {type=double, default=0}
 IsPistol {type=double, default=0}
 IsRifle {type=double, default=0}
]
